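# Niagara system for the FareSporeCookie emote: an audio-reactive grid of meshes on the ground around the player, with glow, lights and rising secondary cubes.
# New_Visualizer treadmills a platform-sized grid with the owner and traces it onto the ground.
# Column height and color follow distance to sampled body joints and smoothed event-audio amplitude, with a looping curve as alternate pulse.
# Time of day scales the lights; front-end and owner velocity adjust motion.
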NiagaraSystem /BRCosmetics/Animation/Game/MainPlayer/Emotes/FareSporeCookie/FX/NS_FareSporeCookie_Final
dependencies {
 usedDataInterfaces [
  &NiagaraDataInterfaceArrayFloat3
  &NiagaraDataInterfaceCollisionQuery
  &NiagaraDataInterfaceDebugDraw
  &NiagaraDataInterfaceArrayFloat
  &NiagaraDataInterfaceCurve
  &NiagaraDataInterfaceVectorCurve
  &NiagaraDataInterfacePlatformSet
  &NiagaraDataInterfaceSocketReader
 ]
 usedDynamicInputs [&SimulationPosition]
 usedModules [
  &SystemState
  &self:SystemSpawnScript.SystemScriptSource.SystemScriptGraph.NiagaraNodeAssignment_0.SetVariables_#1
 ]
}
emitters {
 Glow {
  emitterUpdateScript [
   &EmitterState(
    Inactive Response=::NewEnumerator1
    Life Cycle Mode=::NewEnumerator1
    Loop Behavior=::NewEnumerator1
    Loop Delay=0.2
    Loop Duration Mode=::NewEnumerator1
    UseLoopDelay=true
   )
   &SpawnRate(SpawnRate=2, Use Spawn Probability=false)
   &SpawnBurst_Instantaneous(Use Spawn Probability=false, UseLoopCountLimit=false)
  ]
  particleSpawnScript [
   &InitializeParticle(
    Color=<rgba 0.0626 0.1202 1 1>
    Mass Max=2
    Mass Mode=::NewEnumerator3
    Position Offset Coordinate Space=::Local
    Sprite Rotation Mode=::NewEnumerator2
    Sprite Size Mode=::NewEnumerator3
    Uniform Sprite Size=&RandomRangeFloat(Minimum=150, Maximum=250)
    UsePositionOffset=true
    Write Material Random=true
   )
   &CameraOffset(Camera Offset Amount=-100)
  ]
  particleUpdateScript [
   &ParticleState()
   &SkeletalMeshLocation(
    Bone Sampling Mode=::NewEnumerator4
    MaskSpawnGroup=false
    Orientation Sampling=::NewEnumerator1
    Sampled Position Offset Coordinate Space:ENiagaraSkelMeshTransforms=::NewEnumerator0
    Sampled Scale Coordinate Space:ENiagaraCoordinateSpace=::Simulation
    Skeletal Mesh=DataInterface() {propertyValues {
      DefaultMesh=null
      ExcludeBoneName=root
      OptionalComponentClass=null
      PreviewMesh=/Game/Characters/Player/Male/Small/Bodies/M_Sml_Ninja_04/Meshes/SK_M_SML_Ninja_04_Body
      SkinningMode=SkinOnTheFly
      SoftSourceActor=null
      SourceMode=AttachParent
      UvSetIndex=0
      WholeMeshLOD=-1
      bExcludeBone=false
      bReadDeformedGeometry=true
      bRequireCurrentFrameData=true
      ComponentTags []
      FilteredBones [pelvis,neck_02]
      FilteredSockets []
      MeshUserParameter {parameter {name=None, typeDefHandle {registeredTypeIndex=1}, varData []}}
      SamplingRegions []
     }}
   )
   &ScaleColor(
    Scale Alpha=&FloatFromCurve(
     FloatCurve=DataInterface() {propertyValues {Curve {keys [
         <at 0 0 RCIM_Cubic>
         {arriveTangent=0.1895, interpMode=RCIM_Cubic, leaveTangent=0.1895, tangentMode=RCTM_User, time=0.1, value=1}
         {arriveTangent=-0.0202, interpMode=RCIM_Cubic, leaveTangent=-0.0202, tangentMode=RCTM_User, time=0.7, value=1}
         <at 1.0011 0.002 RCIM_Cubic>
        ]}}}
     Scale Curve=10
    )
    ScaleA=true
    ScaleRGB=true
   )
   &ScaleSpriteSize(Scale Factor=&Vector2DFromFloat(Value=&FloatFromCurve(FloatCurve=DataInterface() {propertyValues {Curve {keys [
          {interpMode=RCIM_Cubic, tangentMode=RCTM_User, time=0, value=0}
          {arriveTangent=-0.0009, interpMode=RCIM_Cubic, leaveTangent=-0.0009, tangentMode=RCTM_User, time=0.2, value=1}
          {tangentMode=RCTM_User, time=1, value=0}
         ]}}})))
   /SetVariables(
    Particles.AirGlowColor:LinearColor=&MultiplyLinearColor(
     Linear Color=$Particles.Color
     Scale Factor=&MakeVector4_FromLinearColor(LinearColor=&MultiplyLinearColor(
       Linear Color=&FNTimeOfDayColor(Night Color=<rgba 1 1 1 0.8>, Morning Color=<rgba 1 1 1 0.5>, Day Color=<rgba 1 1 1 1>, Evening Color=<rgba 1 1 1 0.5>)
       Scale Factor=&MakeVector4_FromLinearColor(LinearColor=<rgba 1 1 1 0.005>)
      ))
    )
    Particles.AirGlowSize:Vector2f=&Multiply_Vector2(A=$Particles.SpriteSize, B=<4 4>)
   ) {isSetParameters=true}
   &FN_Light_Attributes(
    Color=<rgba 1 1 1 0.2>
    Create Color Attribute=true
    Create Exponent Attribute=true
    Create Position Attribute=true
    Create Radius Attribute=true
    Enable Particle Light Enable=true
    Light Enabled=true
    Position=/Niagara/DynamicInputs/Lerp/Lerp_Position(
     A:NiagaraPosition=$Particles.SampleSkeletalMesh001.SampledPosition
     B:NiagaraPosition=/Niagara/DynamicInputs/Helpers/SimulationPosition()
     Alpha:NiagaraFloat=0.5
    )
    Radius=300
    Scale Color=&FloatFromCurve(
     FloatCurve=DataInterface() {propertyValues {Curve {keys [
         {arriveTangent=9.0725, interpMode=RCIM_Cubic, leaveTangent=9.0725, tangentMode=RCTM_User, time=0, value=0}
         {arriveTangent=-2.5249, interpMode=RCIM_Cubic, leaveTangent=-2.5249, tangentMode=RCTM_User, time=0.1408, value=0.9847}
         {arriveTangent=-0.0405, interpMode=RCIM_Cubic, leaveTangent=-0.0405, tangentMode=RCTM_User, time=1, value=0}
        ]}}}
     Scale Curve=10
    )
   )
   /SetVariables(Particles.TempPosition:Vector3f=$Particles.FN_Light_Attributes.Position) {isSetParameters=true}
   &FN_Gen5_Light(
    Attenutation=600
    Exponent=8
    Intensity:NiagaraFloat=&Multiply_Float(A=2, B=&FNTimeOfDayScalar(Night Scale=2, Morning Scale=6, Day Scale=8, Evening Scale=5))
   )
  ]
  rendererClasses [&NiagaraSpriteRendererProperties]
  renderers [{rendererClass=&NiagaraSpriteRendererProperties}]
 }
 Light_Down {
  emitterUpdateScript [
   &EmitterState(
    Inactive Response=::NewEnumerator1
    Life Cycle Mode=::NewEnumerator1
    Loop Behavior=::NewEnumerator1
    Loop Delay=0.2
    Loop Duration Mode=::NewEnumerator1
    UseLoopDelay=true
   )
   &SpawnBurst_Instantaneous(Use Spawn Probability=false, UseLoopCountLimit=false)
  ]
  particleSpawnScript [&InitializeParticle(
    Color=<rgba 1.276 3.8972 5 1>
    Mass Max=1.5
    Mass Mode=::NewEnumerator3
    Position Offset=<0 0 70>
    Sprite Rotation Mode=::NewEnumerator2
    Sprite Size Mode=::NewEnumerator1
    Uniform Sprite Size Max=180
    Uniform Sprite Size Min=150
    UsePositionOffset=true
    Write Material Random=true
   )]
  particleUpdateScript [
   &ParticleState(Kill Particles When Lifetime Has Elapsed=false, Let Infinitely Lived Particles Die When Emitter Deactivates=true)
   /SetVariables(Particles.LerpedAudioReactive:NiagaraFloat=&self:Lerp_Audio_Reactive_EQ(
     Default:Vector3f=&VectorFromFloat(Value=$NPC.FortniteNPC.EventAudio_Amplitude)
     ChangeRate:NiagaraFloat=15
    )) {isSetParameters=true}
   &FN_Light_Attributes(
    Color=&LerpLinearColors(
     StartColor=<rgba 0.0501 0.1468 0.3698 1>
     EndColor=<rgba 0.2188 0.4465 1 1>
     LerpFactor=&ClampFloat(Float=&RemapRange(Input Value To Remap=$Particles.LerpedAudioReactive, Input Min=0.6, Input Max=1, Output Min=0, Output Max=1))
    )
    Create Color Attribute=true
    Create Exponent Attribute=true
    Create Radius Attribute=true
    Exponent=5
    Radius=10
    Scale Color=&FNTimeOfDayScalar(Night Scale=0.05, Morning Scale=0.08, Day Scale=0.1, Evening Scale=0.08)
   )
   &FN_Gen5_Light(Exponent=0.25, Intensity=15)
  ]
  rendererClasses [&NiagaraComponentRendererProperties]
  renderers [{rendererClass=&NiagaraComponentRendererProperties, rendererIndex=0}]
 }
 Light_Up {
  emitterUpdateScript [
   &EmitterState(
    Inactive Response=::NewEnumerator1
    Life Cycle Mode=::NewEnumerator1
    Loop Behavior=::NewEnumerator1
    Loop Delay=0.2
    Loop Duration Mode=::NewEnumerator1
    UseLoopDelay=true
   )
   &SpawnBurst_Instantaneous(Use Spawn Probability=false, UseLoopCountLimit=false)
  ]
  particleSpawnScript [&InitializeParticle(
    Color=<rgba 1.276 3.8972 5 1>
    Mass Max=1.5
    Mass Mode=::NewEnumerator3
    Sprite Rotation Mode=::NewEnumerator2
    Sprite Size Mode=::NewEnumerator1
    Uniform Sprite Size Max=180
    Uniform Sprite Size Min=150
    UsePositionOffset=true
    Write Material Random=true
   )]
  particleUpdateScript [
   &ParticleState(Kill Particles When Lifetime Has Elapsed=false, Let Infinitely Lived Particles Die When Emitter Deactivates=true)
   /SetVariables(Particles.LerpedAudioReactive:NiagaraFloat=&self:Lerp_Audio_Reactive_EQ(
     Default:Vector3f=&VectorFromFloat(Value=$NPC.FortniteNPC.EventAudio_Amplitude)
     ChangeRate:NiagaraFloat=15
    )) {isSetParameters=true}
   &FN_Light_Attributes(
    Color=&LerpLinearColors(
     StartColor=<rgba 0.0501 0.1468 0.3698 1>
     EndColor=<rgba 0.1823 0.4206 1 1>
     LerpFactor=&ClampFloat(Float=&RemapRange(Input Value To Remap=$Particles.LerpedAudioReactive, Input Min=0.6, Input Max=1, Output Min=0, Output Max=1))
    )
    Create Color Attribute=true
    Create Exponent Attribute=true
    Create Radius Attribute=true
    Exponent=1
    Radius=10
    Scale Color=&FNTimeOfDayScalar(Night Scale=0.05, Morning Scale=0.05, Day Scale=0.1, Evening Scale=0.05)
   )
   &FN_Gen5_Light(Exponent=0.25, Intensity=50)
  ]
  rendererClasses [&NiagaraComponentRendererProperties]
  renderers [{rendererClass=&NiagaraComponentRendererProperties, rendererIndex=0}]
 }
 New_Visualizer {
  emitterUpdateScript [
   &EmitterState(
    Inactive Response=::NewEnumerator1
    Life Cycle Mode=::NewEnumerator1
    Loop Behavior=::NewEnumerator1
    Loop Duration Mode=::NewEnumerator1
    UseLoopDelay=false
   )
   /SetVariables(
    Emitter.OriginOffset:Vector3f=&TransformVector(Vector=<0 0 0>, Destination Space=::World)
    Emitter.PaddingPerCell:NiagaraFloat=&MakeCustomFloatFromBool(
     Bool=&IsPlatformSetActive(New Platform Set=DataInterface() {propertyValues {Platforms {qualityLevelMask=-29}}})
     True Float=4
     False Float=5.5
    )
    Emitter.VolumeRadius=110
   ) {isSetParameters=true, module=SetVariables_#2}
   /Niagara/Modules/Spawn/Location/SpawnParticlesInGrid(
    X Count:NiagaraInt32=&MakeCustomIntFromBool(
     Bool=&IsPlatformSetActive(New Platform Set=DataInterface() {propertyValues {Platforms {qualityLevelMask=-29}}})
     True INT=11
     False INT=14
    )
    Y Count:NiagaraInt32=&MakeCustomIntFromBool(
     Bool=&IsPlatformSetActive(New Platform Set=DataInterface() {propertyValues {Platforms {qualityLevelMask=-29}}})
     True INT=11
     False INT=14
    )
    Z Count=0
   )
  ]
  particleSpawnScript [
   &InitializeParticle(
    Color=&Multiply_LinearColorByFloat(Linear Color=<rgba 1 1 1 0.08>, Float=&FN_FrontEndFloatModifier(Gameplay=1, Front End=1))
    Mesh Scale=&MakeCustomVectorFromBool(
     Bool=&IsPlatformSetActive(New Platform Set=DataInterface() {propertyValues {Platforms {qualityLevelMask=-29}}})
     True Vector=<0.27 0.27 0.1>
     False Vector=<0.2 0.2 0.1>
    )
    Mesh Scale Mode=::NewEnumerator4
    Position Offset Coordinate Space=::Local
    Sprite Size Mode=::NewEnumerator3
    UsePositionOffset=true
    Write Material Random=true
   )
   /Niagara/Modules/Spawn/Location/V2/GridLocation(
    Coordinate Space:ENiagaraCoordinateSpace=::World
    Module.Jitter Inside Grid Cells=false
    Module.Override Random Seed=false
    Normalize Offsets=false
    XYZ Dimensions:Vector3f=&MakeVector(
     X=&Divide_Float(A=$Emitter.VolumeRadius, B=$Emitter.PaddingPerCell)
     Y=&Divide_Float(A=$Emitter.VolumeRadius, B=$Emitter.PaddingPerCell)
     Z=0
    )
   )
   &InitialMeshOrientation(
    Mesh Orientation Mode=::NewEnumerator2
    Rotation=&SelectVectorFromArray(
     Array Sampling Mode=::NewEnumerator0
     Direct Array Index=?
     Vector Selection Array=DataInterface() {propertyValues {FloatData [<0 0 0>, <0 0 0.25>, <0 0 0.5>]}}
     Random Seed=0
    )
    Use Rotation Vector=true
   )
   /SetVariables() {isSetParameters=true}
  ]
  particleUpdateScript [
   &ParticleState(Kill Particles When Lifetime Has Elapsed=false, Let Infinitely Lived Particles Die When Emitter Deactivates=true)
   &ScaleColor(
    Scale Alpha=&FN_FrontEndFloatModifier(Gameplay=1, Front End=1)
    Scale RGB=&Multiply_VectorByFloat(Vector=<1 1 1>, Float=&FN_FrontEndFloatModifier(Gameplay=1, Front End=1))
    ScaleA=true
    ScaleRGB=true
   ) {module=ScaleColor001}
   &SkeletalMeshLocation(
    Bone Sampling Mode=::NewEnumerator4
    MaskSpawnGroup=false
    Orientation Sampling=::NewEnumerator1
    Position Sampling=::NewEnumerator2
    Sampled Orientation Coordinate Space=::World
    Sampled Position Coordinate Space=::World
    Sampled Position Offset Coordinate Space:ENiagaraSkelMeshTransforms=::NewEnumerator0
    Sampled Scale Coordinate Space:ENiagaraCoordinateSpace=::Simulation
    Skeletal Mesh=DataInterface() {propertyValues {
      DefaultMesh=null
      ExcludeBoneName=root
      OptionalComponentClass=null
      PreviewMesh=&SK_M_SML_Ninja_04_Body
      SkinningMode=SkinOnTheFly
      SoftSourceActor=null
      SourceMode=AttachParent
      UvSetIndex=0
      WholeMeshLOD=-1
      bExcludeBone=false
      bReadDeformedGeometry=true
      bRequireCurrentFrameData=true
      ComponentTags []
      FilteredBones [pelvis,neck_02]
      FilteredSockets []
      MeshUserParameter {parameter {name=None, typeDefHandle {registeredTypeIndex=1}, varData []}}
      SamplingRegions []
     }}
   )
   /SetVariables(Particles.PelvisPosition:NiagaraPosition=$Output.SkeletalMeshLocation.SampledPositionWithoutOffset) {isSetParameters=true, module=SetVariables_#3}
   &SkeletalMeshLocation(
    Bone / Socket Index=1
    Bone Sampling Mode=::NewEnumerator4
    MaskSpawnGroup=false
    Orientation Sampling=::NewEnumerator1
    Position Sampling=::NewEnumerator2
    Sampled Position Coordinate Space=::World
    Sampled Position Offset Coordinate Space:ENiagaraSkelMeshTransforms=::NewEnumerator0
    Sampled Scale Coordinate Space:ENiagaraCoordinateSpace=::Simulation
    Skeletal Mesh=DataInterface() {propertyValues {
      DefaultMesh=null
      ExcludeBoneName=root
      OptionalComponentClass=null
      PreviewMesh=&SK_M_SML_Ninja_04_Body
      SkinningMode=SkinOnTheFly
      SoftSourceActor=null
      SourceMode=AttachParent
      UvSetIndex=0
      WholeMeshLOD=-1
      bExcludeBone=false
      bReadDeformedGeometry=true
      bRequireCurrentFrameData=false
      ComponentTags []
      FilteredBones [pelvis,neck_02]
      FilteredSockets []
      MeshUserParameter {parameter {name=None, typeDefHandle {registeredTypeIndex=1}, varData []}}
      SamplingRegions []
     }}
   ) {module=SkeletalMeshLocation001}
   /SetVariables(Particles.NeckPosition:NiagaraPosition=$Output.SkeletalMeshLocation001.SampledPositionWithoutOffset) {isSetParameters=true, module=SetVariables_#4}
   &SkeletalMeshLocation(
    Bone / Socket Index=2
    Bone Sampling Mode=::NewEnumerator4
    MaskSpawnGroup=false
    Orientation Sampling=::NewEnumerator1
    Position Sampling=::NewEnumerator2
    Sampled Position Offset Coordinate Space:ENiagaraSkelMeshTransforms=::NewEnumerator0
    Sampled Scale Coordinate Space:ENiagaraCoordinateSpace=::Simulation
    Skeletal Mesh=DataInterface() {propertyValues {
      DefaultMesh=null
      ExcludeBoneName=root
      OptionalComponentClass=null
      PreviewMesh=&SK_M_SML_Ninja_04_Body
      SkinningMode=SkinOnTheFly
      SoftSourceActor=null
      SourceMode=AttachParent
      UvSetIndex=0
      WholeMeshLOD=-1
      bExcludeBone=false
      bReadDeformedGeometry=true
      bRequireCurrentFrameData=true
      ComponentTags []
      FilteredBones [pelvis,neck_02,foot_l,ik_foot_l]
      FilteredSockets []
      MeshUserParameter {parameter {name=None, typeDefHandle {registeredTypeIndex=1}, varData []}}
      SamplingRegions []
     }}
   ) {module=SkeletalMeshLocation002}
   /SetVariables(Particles.LeftLegPosition:NiagaraPosition=$Output.SkeletalMeshLocation002.SampledPositionWithoutOffset) {isSetParameters=true, module=SetVariables_#5}
   &SkeletalMeshLocation(
    Bone / Socket Index=3
    Bone Sampling Mode=::NewEnumerator4
    MaskSpawnGroup=false
    Orientation Sampling=::NewEnumerator1
    Position Sampling=::NewEnumerator2
    Sampled Position Offset Coordinate Space:ENiagaraSkelMeshTransforms=::NewEnumerator0
    Sampled Scale Coordinate Space:ENiagaraCoordinateSpace=::Simulation
    Skeletal Mesh=DataInterface() {propertyValues {
      DefaultMesh=null
      ExcludeBoneName=root
      OptionalComponentClass=null
      PreviewMesh=&SK_M_SML_Ninja_04_Body
      SkinningMode=SkinOnTheFly
      SoftSourceActor=null
      SourceMode=AttachParent
      UvSetIndex=0
      WholeMeshLOD=-1
      bExcludeBone=false
      bReadDeformedGeometry=true
      bRequireCurrentFrameData=true
      ComponentTags []
      FilteredBones [pelvis,neck_02,calf_r,ik_foot_r]
      FilteredSockets []
      MeshUserParameter {parameter {name=None, typeDefHandle {registeredTypeIndex=1}, varData []}}
      SamplingRegions []
     }}
   ) {module=SkeletalMeshLocation003}
   /SetVariables(Particles.RightLegPosition:NiagaraPosition=$Output.SkeletalMeshLocation003.SampledPositionWithoutOffset) {isSetParameters=true, module=SetVariables_#6}
   &SkeletalMeshLocation(
    Bone Sampling Mode=::NewEnumerator4
    MaskSpawnGroup=false
    Orientation Sampling=::NewEnumerator1
    Position Sampling=::NewEnumerator2
    Sampled Position Offset Coordinate Space:ENiagaraSkelMeshTransforms=::NewEnumerator0
    Sampled Scale Coordinate Space:ENiagaraCoordinateSpace=::Simulation
    Skeletal Mesh=DataInterface() {propertyValues {
      DefaultMesh=null
      ExcludeBoneName=root
      OptionalComponentClass=null
      PreviewMesh=&SK_M_SML_Ninja_04_Body
      SkinningMode=SkinOnTheFly
      SoftSourceActor=null
      SourceMode=AttachParent
      UvSetIndex=0
      WholeMeshLOD=-1
      bExcludeBone=false
      bReadDeformedGeometry=true
      bRequireCurrentFrameData=true
      ComponentTags []
      FilteredBones [calf_l]
      FilteredSockets []
      MeshUserParameter {parameter {name=None, typeDefHandle {registeredTypeIndex=1}, varData []}}
      SamplingRegions []
     }}
   ) {module=SkeletalMeshLocation004}
   /SetVariables(Particles.LeftKneePosition:NiagaraPosition=$Particles.SkeletalMeshLocation004.SampledPosition) {isSetParameters=true, module=SetVariables_#7}
   &SkeletalMeshLocation(
    Bone Sampling Mode=::NewEnumerator4
    MaskSpawnGroup=false
    Orientation Sampling=::NewEnumerator1
    Position Sampling=::NewEnumerator2
    Sampled Position Offset Coordinate Space:ENiagaraSkelMeshTransforms=::NewEnumerator0
    Sampled Scale Coordinate Space:ENiagaraCoordinateSpace=::Simulation
    Skeletal Mesh=DataInterface() {propertyValues {
      DefaultMesh=null
      ExcludeBoneName=root
      OptionalComponentClass=null
      PreviewMesh=&SK_M_SML_Ninja_04_Body
      SkinningMode=SkinOnTheFly
      SoftSourceActor=null
      SourceMode=AttachParent
      UvSetIndex=0
      WholeMeshLOD=-1
      bExcludeBone=false
      bReadDeformedGeometry=true
      bRequireCurrentFrameData=true
      ComponentTags []
      FilteredBones [calf_r]
      FilteredSockets []
      MeshUserParameter {parameter {name=None, typeDefHandle {registeredTypeIndex=1}, varData []}}
      SamplingRegions []
     }}
   ) {module=SkeletalMeshLocation005}
   /SetVariables(Particles.RightKneePosition:NiagaraPosition=$Output.SkeletalMeshLocation005.SampledPositionWithoutOffset) {isSetParameters=true, module=SetVariables_#8}
   &SolveForcesAndVelocity(Clamp Velocity=false, Limit Acceleration=false)
   /SetVariables(Particles.PositionCenter:NiagaraPosition=&AddVectorToPosition(Position=$Engine.Owner.Position, Vector=$Emitter.OriginOffset)) {isSetParameters=true, module=SetVariables_#9}
   &self:TreadmillParticles_01(
    GridSize:NiagaraFloat=&Multiply_Float(
     A=$Emitter.VolumeRadius
     B=&Lerp_Float(
      A=2.5
      B=2.5
      Alpha=&MakeCustomFloatFromBool(
       Bool=&SetBoolByFloatComparison(A=&VectorLength(Vector=$User.OwnerLocalVelocity), B=160)
       True Float=2
       False Float=0
      )
     )
    )
    InputPosition:NiagaraPosition=$Particles.PositionCenter
   ) {module=TreadmillParticles_01}
   /SetVariables(Particles.LerpedAudioReactive:NiagaraFloat=&self:Lerp_Audio_Reactive_EQ(
     Default:Vector3f=&VectorFromFloat(Value=&Lerp_Float(
       A=$NPC.FortniteNPC.EventAudio_Amplitude
       B=&FloatFromCurve(
        FloatCurve=DataInterface() {propertyValues {Curve {defaultValue=1, keys [<at 0 1>,<at 1 0.3>]}}}
        CurveIndex=&ModuloFloat(A=$Engine.System.Age, B=0.4688)
        Scale Curve=0.8
       )
       Alpha=&MakeCustomFloatFromBool(
        Bool=&SetBoolByFloatComparison(A=$NPC.FortniteNPC.EventAudio_Amplitude, B=0.3, Comparison Type=::NewEnumerator6)
        True Float=1
        False Float=0
       )
      ))
     ChangeRate:NiagaraFloat=15
    )) {isSetParameters=true, module=SetVariables_#10}
   &self:GridCollision(TraceDir:Vector3f=<0 0 -1>, TraceDist=250, bIsFrontEnd:NiagaraBool=$User.bIsFrontEndPreview) {enabled=false, module=GridCollision}
   /SetVariables(Particles.TargetPos:Vector3f=&self:Return_Closest_Joint(
     PelvisPosition:Vector3f=$Particles.PelvisPosition
     NeckPosition:Vector3f=$Particles.NeckPosition
     LeftLegPosition:Vector3f=$Particles.LeftLegPosition
     RightLegPosition:Vector3f=$Particles.RightLegPosition
     LeftCalfPosition:Vector3f=$Particles.LeftKneePosition
     RightCalfPosition:Vector3f=$Particles.RightKneePosition
     NeckOffset:Vector3f=&Lerp_Vector(
      A=<0 0 -60>
      B=<0 0 -20>
      Alpha=&RemapRange(
       Input Value To Remap=&ClampFloat(Float=$Engine.System.Age, Max=0.8)
       Input Min=0
       Input Max=0.8
       Output Min=0
       Output Max=1
      )
     )
     PelvisOffset:Vector3f=&Lerp_Vector(
      A=<0 0 -60>
      B=<0 0 -10>
      Alpha=&RemapRange(
       Input Value To Remap=&ClampFloat(Float=$Engine.System.Age, Max=0.8)
       Input Min=0
       Input Max=0.8
       Output Min=0
       Output Max=1
      )
     )
     KneeOffset:Vector3f=<0 0 -20>
     FootOffset:Vector3f=<0 0 -20>
     ParticlePosition:Vector3f=Linked() {linkedVariable {name=Particles.Position, type=NiagaraPosition}}
    )) {isSetParameters=true, module=SetVariables_#11}
   &RegionMask(
    Invert Region=true
    Offset:Vector3f=<0 165 0>
    Offset Space:ENiagaraCoordinateSpace=::Local
    Outer Float=0
    Region Dimensions:Vector3f=<200 850 1500>
    Region Falloff=200
    Region Type:ENiagaraDebugDrawShape=::NewEnumerator1
    Rotation:Vector3f=$Engine.Owner.SystemYAxis
    Rotation Angle Type:ENiagara_AngleInput=::NewEnumerator0
    UseOffset=true
    UseRotation=true
   )
   &Timeline(
    Additional Output:ENiagara_TimelineOutput=::NewEnumerator1
    Additional Output Mode:ENiagara_TimelineOutputMode=::NewEnumerator1
    Float Curve=DataInterface() {propertyValues {Curve {
       defaultValue=0
       keys [
        {interpMode=RCIM_Cubic, tangentMode=RCTM_User, time=0, value=0}
        {arriveTangent=1, interpMode=RCIM_Cubic, leaveTangent=1, time=0.2, value=0}
        {interpMode=RCIM_Cubic, tangentMode=RCTM_User, time=1, value=1}
       ]
      }}}
    Play=&Boolean_NotOperation(Bool=$Particles.RegionMask.IsInsideMask)
    Play Rate=&Lerp_Float(
     A=5
     B=20
     Alpha=&RemapRange(
      Input Value To Remap=&VectorLength(Vector=$User.OwnerLocalVelocity)
      Input Min=0
      Input Max=800
      Output Min=0
      Output Max=1
     )
    )
    Rewind Rate:NiagaraFloat=&Lerp_Float(
     A=5
     B=20
     Alpha=&RemapRange(
      Input Value To Remap=&VectorLength(Vector=$User.OwnerLocalVelocity)
      Input Min=0
      Input Max=800
      Output Min=0
      Output Max=1
     )
    )
   )
   &SocketLocation(
    Orientation Sampling=::NewEnumerator1
    Position Sampling=::NewEnumerator1
    SamplePosOffset=false
    Socket Reader=$System.Socket Reader
    Socket Sampling Mode=::NewEnumerator3
    VelocitySpeedLimit=false
   )
   &DoOnce(
    Reset:NiagaraBool=&SetBoolByFloatComparison(A=$Particles.Timeline001.Time, B=0.3)
    Trigger Condition:NiagaraBool=&SetBoolByFloatComparison(
     A=&VectorLength(Vector=Linked() {linkedVariable {name=Particles.Position, type=NiagaraPosition}})
     B=&VectorLength(Vector=Linked() {linkedVariable {name=Particles.Previous.Position, type=NiagaraPosition}})
     Comparison Type=::NewEnumerator5
    )
   )
   &Timeline(Instantaneous Rewind=true, Loop While Playing=true, Play=$Particles.DoOnce.TriggeredOnce, Play Rate=10) {module=Timeline001}
   &self:GridCollision_01(
    CharacterPos:NiagaraPosition=&AddVectorToPosition(Position=$Engine.Owner.Position, Vector=<0 0 50>)
    Disable Trace:NiagaraBool=&Boolean_AndOperation(
     Bool1=&Boolean_NotOperation(Bool=$Particles.DoOnce.TriggeredOnce)
     Bool2=&SetBoolByFloatComparison(A=$Emitter.Age, B=0.4)
    )
    Height Adjustment=1.5
    Trace Origin Height=120
    TraceDir:Vector3f=<0 0 -1>
    TraceDist=250
    bIsFrontEnd:NiagaraBool=$User.bIsFrontEndPreview
   ) {module=GridCollision_01}
   /SetVariables(Particles.Position=/Niagara/DynamicInputs/TypeConversions/MakeCustomPositionFromBool(
     Bool:NiagaraBool=&Boolean_OrOperation(
      Bool1=$User.bIsFrontEndPreview
      Bool2=&SetBoolByFloatComparison(A=&VectorLength(Vector=$Engine.Owner.Velocity), B=300)
     )
     True Position:NiagaraPosition=&AddVectorToPosition(
      Position=&MakeVector(
       X=&MakeFloatFromVector(Vector=Linked() {linkedVariable {name=Particles.Position, type=NiagaraPosition}}, Channel=::NewEnumerator0)
       Y=&MakeFloatFromVector(Vector=Linked() {linkedVariable {name=Particles.Position, type=NiagaraPosition}}, Channel=::NewEnumerator1)
       Z=&MakeFloatFromVector(Vector=Linked() {linkedVariable {name=Output.SocketLocation.SampledPosition, type=NiagaraPosition}}, Channel=::NewEnumerator2)
      )
      Vector=&FN_FrontEndVector3Modifier(
       Gameplay=<0 0 0>
       Front End=&RotateVectorByQuaternion(
        Vector To Rotate=&Multiply_VectorByFloat(Vector=<0 -50 0>, Float=$Engine.DeltaTime)
        Quaternion=$Output.SocketLocation.SampledRotation
       )
      )
     )
     False Position:NiagaraPosition=&MakeCustomPositionFromBool(
      Bool:NiagaraBool=$Particles.DoOnce.TriggeredOnce
      True Position:NiagaraPosition=$Particles.CollisionPos
      False Position:NiagaraPosition=&MakeCustomPositionFromBool(
       Bool:NiagaraBool=&SetBoolByFloatComparison(A=$Emitter.Age, B=1)
       True Position:NiagaraPosition=$Particles.Previous.Position
       False Position:NiagaraPosition=$Particles.CollisionPos
      )
     )
    )) {isSetParameters=true}
   &ScaleMeshSize(Scale Factor=&VectorFromFloat(Value=&Multiply_Float(
      A=$Particles.Timeline.Float
      B=&FloatFromCurve(FloatCurve=DataInterface() {propertyValues {Curve {
          defaultValue=1
          keys [{interpMode=RCIM_Cubic, tangentMode=RCTM_User, time=0, value=1}, {interpMode=RCIM_Cubic, tangentMode=RCTM_User, time=1, value=1}]
         }}})
     ))) {module=ScaleMeshSize001}
   /SetVariables(Particles.ParticleZScaleFactor:NiagaraFloat=&ClampFloat(
     Float=&Multiply_Float(
      A=&self:Lerp_Player_Position_EQ(
       TargetPos:Vector3f=$Particles.TargetPos
       ChangeRate:NiagaraFloat=20
       XY Distance:NiagaraFloat=100
       ParticlePosition:Vector3f=Linked() {linkedVariable {name=Particles.CollisionPos, type=NiagaraPosition}}
       BaseScale:NiagaraFloat=&Lerp_Float(
        A=-200
        B=125
        Alpha=&RemapRange(Input Value To Remap=$Particles.LerpedAudioReactive, Input Min=0.3, Input Max=1, Output Min=0, Output Max=1)
       )
       InputMinTest:NiagaraFloat=1
       TargetZOffset:NiagaraFloat=0
       DeltaTime:NiagaraFloat=&Multiply_Float(A=$Engine.DeltaTime, B=1)
      )
      B=&Lerp_Float(
       A=0.1
       B=0.12
       Alpha=&RemapRange(
        Input Value To Remap=&VectorLength(Vector=$User.OwnerLocalVelocity)
        Input Min=0
        Input Max=800
        Output Min=0
        Output Max=1
       )
      )
     )
     Min=1
     Max=&Lerp_Float(
      A=0
      B=35
      Alpha=&RemapRange(
       Input Value To Remap=&ClampFloat(Float=$Engine.System.Age, Max=0.5)
       Input Min=0
       Input Max=0.5
       Output Min=0
       Output Max=1
      )
     )
    )) {isSetParameters=true, module=SetVariables_#12}
   &ScaleMeshSize(Scale Factor=&MakeVector(X=1, Y=1, Z=$Particles.ParticleZScaleFactor)) {module=ScaleMeshSize002}
   &ScaleMeshSize(Scale Factor=&Lerp_Vector(
     A=<1 1 1>
     B=<1 1 0.5>
     Alpha=&RemapRange(
      Input Value To Remap=&self:XY_Distance_From_Closest_Joint(Target:Vector3f=$Particles.TargetPos, Current:Vector3f=Linked() {linkedVariable {name=Particles.Position, type=NiagaraPosition}})
      Input Min=0
      Input Max=400
      Output Min=0
      Output Max=1
     )
    )) {module=ScaleMeshSize008}
   /SetVariables(Particles.AudioReactiveColorValue:NiagaraFloat=&Multiply_Float(
     A=&Lerp_Float(
      A=1
      B=0.5
      Alpha=&RemapRange(
       Input Value To Remap=&self:XY_Distance_From_Closest_Joint(Target:Vector3f=$Particles.TargetPos, Current:Vector3f=Linked() {linkedVariable {name=Particles.Position, type=NiagaraPosition}})
       Input Min=0
       Input Max=400
       Output Min=0
       Output Max=1
      )
     )
     B=&Lerp_Float(
      A=-0.8
      B=1
      Alpha=&RemapRange(Input Value To Remap=$Particles.LerpedAudioReactive, Input Min=0.3, Input Max=1, Output Min=0, Output Max=1)
     )
    )) {isSetParameters=true, module=SetVariables_#13}
   &ScaleColor(
    Scale Alpha=&RemapRange(
     Input Value To Remap=&DistanceBetweenPositions(A=$Particles.Position, B=Linked() {linkedVariable {name=Particles.PelvisPosition, type=Vector3f}})
     Input Min=70
     Input Max=110
     Output Min=0
     Output Max=1
    )
    Scale RGB=&Lerp_Vector(
     A=&Lerp_Vector(
      A=&FN_FrontEndVector3Modifier(
       Gameplay=&MakeVectorFromLinearColorRGB(Color=<rgba 0.1048 5 1.9461 1>)
       Front End=&MakeVectorFromLinearColorRGB(Color=<rgba 0 1.5576 3 1>)
      )
      B=&MakeVectorFromLinearColorRGB(Color=<rgba 0 2.5179 5 1>)
      Alpha=&RemapRange(
       Input Value To Remap=&Sine(
        Normalized Angle=&Add_Float(A=$Emitter.Age, B=&RandomRangeFloat(Minimum=-1, Maximum=1))
        Period=5
        Scale=1
        Bias=0
       )
       Input Min=-1
       Input Max=1
       Output Min=0
       Output Max=0.6
      )
     )
     B=&MakeVectorFromLinearColorRGB(Color=<rgba 0.8642 0.7917 1 1>)
     Alpha=&RemapRange(
      Input Value To Remap=&DistanceBetweenPositions(
       A=&MakeVector(
        X=&MakeFloatFromVector(Vector=Linked() {linkedVariable {name=Particles.Position, type=NiagaraPosition}}, Channel=::NewEnumerator0)
        Y=&MakeFloatFromVector(Vector=Linked() {linkedVariable {name=Particles.Position, type=NiagaraPosition}}, Channel=::NewEnumerator1)
        Z=-2
       )
       B=&MakeVector(
        X=&MakeFloatFromVector(Vector=&SimulationPosition(), Channel=::NewEnumerator0)
        Y=&MakeFloatFromVector(Vector=&SimulationPosition(), Channel=::NewEnumerator1)
        Z=0
       )
      )
      Input Min=0
      Input Max=&Lerp_Float(
       A=50
       B=120
       Alpha=&RemapRange(Input Value To Remap=$Particles.LerpedAudioReactive, Input Min=0.6, Input Max=1, Output Min=0, Output Max=1)
      )
      Output Min=0
      Output Max=1
     )
    )
    ScaleA=true
    ScaleRGB=true
   ) {module=ScaleColor002}
   &ScaleColor(
    Scale Alpha=&RemapRange(
     Input Value To Remap=&DistanceBetweenPositions(A=$Particles.Position, B=Linked() {linkedVariable {name=Particles.PelvisPosition, type=Vector3f}})
     Input Min=70
     Input Max=110
     Output Min=0
     Output Max=1
    )
    Scale RGB=&Lerp_Vector(
     A=<1 1 1>
     B=&MakeVectorFromLinearColorRGB(Color=<rgba 0.0968 4.6189 1.7978 1>)
     Alpha=&ClampFloat(Float=&Multiply_Float(A=$Particles.AudioReactiveColorValue, B=1))
    )
    ScaleA=true
    ScaleRGB=true
   )
   &ScaleMeshSize(Scale Factor=&Lerp_Vector(
     A=<1 1 1>
     B=&Lerp_Vector(
      A=<1 1 0.8>
      B=<1 1 1.1>
      Alpha=&Lerp_Float(
       A=0
       B=&Add_Float(
        A=&Sine(
         Normalized Angle=&Add_Float(
          A=$Emitter.Age
          B=&Multiply_Float(
           A=&DistanceBetweenPositions(A=$Particles.Position, B=Linked() {linkedVariable {name=Particles.PelvisPosition, type=Vector3f}})
           B=-0.01
          )
         )
         Period=1
         Scale=1
         Bias=0
        )
        B=1
       )
       Alpha=&RemapRange(
        Input Value To Remap=&DistanceBetweenPositions(A=Linked() {linkedVariable {name=Particles.PelvisPosition, type=Vector3f}}, B=$Particles.Position)
        Input Min=0
        Input Max=150
        Output Min=0
        Output Max=1
       )
      )
     )
     Alpha=&self:Height_Gate(
      ChangeRate:NiagaraFloat=12
      Max:NiagaraFloat=0.45
      Particle Z Scale:NiagaraFloat=&MakeFloatFromVector(Vector=$Particles.Scale, Channel=::NewEnumerator2)
     )
    ))
  ]
  rendererClasses [&NiagaraMeshRendererProperties, &NiagaraSpriteRendererProperties]
  renderers [
   {rendererClass=&NiagaraMeshRendererProperties, rendererIndex=0}
   {rendererClass=&NiagaraMeshRendererProperties, rendererIndex=1}
   {rendererClass=&NiagaraSpriteRendererProperties, rendererIndex=2}
  ]
 }
 Secondary_Cubes {
  emitterUpdateScript [
   &EmitterState(
    Inactive Response=::NewEnumerator1
    Life Cycle Mode=::NewEnumerator1
    Loop Behavior=::NewEnumerator1
    Loop Duration Mode=::NewEnumerator1
    UseLoopDelay=false
   )
   &SpawnRate(
    SpawnRate=&Multiply_Float(
     A=&FloatFromCurve(
      FloatCurve=DataInterface() {propertyValues {Curve {keys [<at 0 1>,<at 1 0>]}}}
      CurveIndex=$Emitter.NormalizedLoopAge
      Scale Curve=&Lerp_Float(A=0, B=50, Alpha=&ClampFloat(Float=1))
     )
     B=&Lerp_Float(A=1, B=0, Alpha=0)
    )
    Use Spawn Probability=false
   )
  ]
  particleSpawnScript [
   &PartitionParticles()
   &InitializeParticle(
    Color=&Multiply_LinearColorByFloat(
     Linear Color=&RandomRangeLinearColor(Minimum=<rgba 1.1458 3.4402 5 1>, Maximum=<rgba 2.479 1.8354 4 1>)
     Float=&FN_FrontEndFloatModifier(Gameplay=1, Front End=1)
    )
    Mesh Scale=&Multiply_Vector(
     A=<1 1 1>
     B=&VectorFromFloat(Value=&UniformRangedFloat(
       Minimum=&Multiply_Float(
        A=0.15
        B=&FloatFromCurve(
         FloatCurve=DataInterface() {propertyValues {Curve {keys [<at 0 1>,<at 1 1>]}}}
         CurveIndex=&Divide_Float(A=&VectorLength(Vector=$Engine.Owner.Velocity), B=9000)
        )
       )
       Maximum=&Multiply_Float(
        A=0.2
        B=&FloatFromCurve(
         FloatCurve=DataInterface() {propertyValues {Curve {keys [<at 0 1>,<at 1 1.2>]}}}
         CurveIndex=&Divide_Float(A=&VectorLength(Vector=$Engine.Owner.Velocity), B=9000)
        )
       )
      ))
    )
    Mesh Scale Mode=::NewEnumerator4
    UsePositionOffset=false
    Write Material Random=true
   )
   &ShapeLocation(
    Box Size=&Lerp_Vector(
     A=<150 180 10>
     B=<150 200 10>
     Alpha=&MakeCustomFloatFromBool(
      Bool=&SetBoolByFloatComparison(A=&VectorLength(Vector=$User.OwnerLocalVelocity), B=1)
      True Float=1
      False Float=0
     )
    )
    MaskSpawnGroup=false
    Offset=<0 30 0>
    Offset Coordinate Space=::Local
    Offset Mode=::NewEnumerator0
    Override Seed=false
    Shape Primitive=::NewEnumerator2
    Surface Only=false
    UseOwnerScale=true
   )
   &RegionMask(
    Invert Region=true
    Offset:Vector3f=<0 10 0>
    Offset Space:ENiagaraCoordinateSpace=::Local
    Outer Float=0
    Region Dimensions:Vector3f=<225 500 1500>
    Region Falloff=300
    Region Type:ENiagaraDebugDrawShape=::NewEnumerator1
    Rotation:Vector3f=$Engine.Owner.SystemYAxis
    Rotation Angle Type:ENiagara_AngleInput=::NewEnumerator0
    UseOffset=true
    UseRotation=true
   )
   &AddVelocity(Coordinate Space=::World, Velocity=<0 0 100>)
   &AddVelocity(Velocity=&FN_FrontEndVector3Modifier(Gameplay=<0 0 0>, Front End=<0 -100 0>)) {module=AddVelocity001}
  ]
  particleUpdateScript [
   &ParticleState()
   &ScaleColor(
    Scale Alpha=&FN_FrontEndFloatModifier(Gameplay=1, Front End=0.1)
    Scale RGB=&Multiply_VectorByFloat(Vector=<1 1 1>, Float=&FN_FrontEndFloatModifier(Gameplay=1, Front End=2))
    ScaleA=true
    ScaleRGB=true
   ) {module=ScaleColor001}
   &SkeletalMeshLocation(
    Bone Sampling Mode=::NewEnumerator4
    MaskSpawnGroup=false
    Orientation Sampling=::NewEnumerator1
    Position Sampling=::NewEnumerator2
    Sampled Position Offset Coordinate Space:ENiagaraSkelMeshTransforms=::NewEnumerator0
    Sampled Scale Coordinate Space:ENiagaraCoordinateSpace=::Simulation
    Skeletal Mesh=DataInterface() {propertyValues {
      DefaultMesh=null
      ExcludeBoneName=root
      OptionalComponentClass=null
      PreviewMesh=&SK_M_SML_Ninja_04_Body
      SkinningMode=SkinOnTheFly
      SoftSourceActor=null
      SourceMode=AttachParent
      UvSetIndex=0
      WholeMeshLOD=-1
      bExcludeBone=false
      bReadDeformedGeometry=true
      bRequireCurrentFrameData=true
      ComponentTags []
      FilteredBones [pelvis,neck_02]
      FilteredSockets []
      MeshUserParameter {parameter {name=None, typeDefHandle {registeredTypeIndex=1}, varData []}}
      SamplingRegions []
     }}
   )
   /SetVariables(Particles.PelvisPosition:NiagaraPosition=$Output.SkeletalMeshLocation.SampledPositionWithoutOffset) {isSetParameters=true}
   &ScaleMeshSize(Scale Factor=&VectorFromFloat(Value=&MaskFloat(
      Mask Source=1
      Mask Parameter=$Particles.PartitionParticles.Partition
      Mask=0
      Masked Float=&Lerp_Float(A=1, B=0, Alpha=&ClampFloat(Float=$User.AltTimeline))
      Passthrough Float=1
     )))
   &ScaleMeshSize(Scale Factor=&VectorFromCurve(VectorCurve=DataInterface() {propertyValues {XCurve {keys [<at 0 1>,<at 1 0>]}, YCurve {keys [<at 0 1>,<at 1 0>]}, ZCurve {keys [<at 0 1>,<at 1 0>]}}})) {module=ScaleMeshSize001}
   &Drag(Drag=2, Use Linear Drag=true, Use Rotational Drag=false)
   &SolveForcesAndVelocity(Clamp Velocity=false, Limit Acceleration=false)
   &ScaleColor(
    Scale RGB=&FN_FrontEndVector3Modifier(
     Gameplay=&MakeVectorFromLinearColorRGB(Color=<rgba 0.5168 0.5579 0.7 1>)
     Front End=&MakeVectorFromLinearColorRGB(Color=$User.FrontEnd_color)
    )
    ScaleA=true
    ScaleRGB=true
   ) {module=ScaleColor002}
   &ScaleColor(
    Scale RGB=&Lerp_Vector(
     A=&MakeVectorFromLinearColorRGB(Color=<rgba 0.7552 0.7614 1 1>)
     B=&MakeVectorFromLinearColorRGB(Color=<rgba 0.2753 0.4267 1 1>)
     Alpha=&RemapRange(
      Input Value To Remap=&Sine(
       Normalized Angle=&Add_Float(A=$Emitter.Age, B=&RandomRangeFloat(Minimum=-1, Maximum=1))
       Period=5
       Scale=1
       Bias=0
      )
      Input Min=-1
      Input Max=1
      Output Min=0.5
      Output Max=1
     )
    )
    ScaleA=true
    ScaleRGB=true
   )
   /SetVariables(Particles.Color=&LerpLinearColors(
     StartColor=$Particles.Color
     EndColor=&MultiplyLinearColor(
      Linear Color=<rgba 0.3706 0.2595 0.7448 0.4748>
      Scale Factor=&MakeVector4(X=1, Y=1, Z=1, W=&FN_FrontEndFloatModifier(Gameplay=1, Front End=0.5))
     )
     LerpFactor=&RemapRange(
      Input Value To Remap=&DistanceBetweenPositions(
       A=&MakeVector(
        X=&MakeFloatFromVector(Vector=Linked() {linkedVariable {name=Particles.Position, type=NiagaraPosition}}, Channel=::NewEnumerator0)
        Y=&MakeFloatFromVector(Vector=Linked() {linkedVariable {name=Particles.Position, type=NiagaraPosition}}, Channel=::NewEnumerator1)
        Z=0
       )
       B=&MakeVector(
        X=&MakeFloatFromVector(Vector=$Particles.PelvisPosition, Channel=::NewEnumerator0)
        Y=&MakeFloatFromVector(Vector=$Particles.PelvisPosition, Channel=::NewEnumerator1)
        Z=0
       )
      )
      Input Min=40
      Input Max=80
      Output Min=0
      Output Max=1
     )
    )) {isSetParameters=true}
   &ScaleMeshSize(Scale Factor=&VectorFromFloat(Value=&FloatFromCurve(
      FloatCurve=DataInterface() {propertyValues {Curve {keys [
          {arriveTangent=-1.3985, interpMode=RCIM_Cubic, leaveTangent=-1.3985, tangentMode=RCTM_User, time=0, value=1}
          {arriveTangent=-0.5, interpMode=RCIM_Cubic, leaveTangent=-0.5, time=0.1, value=0.6}
          {arriveTangent=-0.5, interpMode=RCIM_Cubic, leaveTangent=-0.5, time=0.3, value=0.85}
          {arriveTangent=-0.1, interpMode=RCIM_Cubic, leaveTangent=-0.1, time=0.5, value=0.4}
          {arriveTangent=-0.375, interpMode=RCIM_Cubic, leaveTangent=-0.375, time=0.8, value=0.8}
          {arriveTangent=1, interpMode=RCIM_Cubic, leaveTangent=1, time=0.9, value=0.25}
          <at 1 1>
         ]}}}
      CurveIndex=&Add_Float(A=$Particles.NormalizedAge, B=&RandomRangeFloat(Minimum=0, Maximum=0.5))
      Scale Curve=1.5
     ))) {module=ScaleMeshSize002}
   &CameraOffset()
  ]
  rendererClasses [&NiagaraMeshRendererProperties]
  renderers [{rendererClass=&NiagaraMeshRendererProperties, rendererIndex=0}]
 }
 Visualizer {
  bEnabled=false
  emitterUpdateScript [
   &EmitterState(
    Inactive Response=::NewEnumerator1
    Life Cycle Mode=::NewEnumerator1
    Loop Behavior=::NewEnumerator1
    Loop Duration Mode=::NewEnumerator1
    UseLoopDelay=false
   )
   /SetVariables(
    Emitter.OriginOffset:Vector3f=&TransformVector(Vector=<0 0 0>, Destination Space=::World)
    Emitter.PaddingPerCell=5.5
    Emitter.VolumeRadius=110
   ) {isSetParameters=true, module=SetVariables_#2}
   &SpawnParticlesInGrid(X Count=14, Y Count=14, Z Count=0)
  ]
  particleSpawnScript [
   &InitializeParticle(
    Color=&Multiply_LinearColorByFloat(Linear Color=<rgba 1 1 1 0.08>, Float=&FN_FrontEndFloatModifier(Gameplay=1, Front End=1))
    Mesh Scale=<0.2 0.2 0.1>
    Mesh Scale Mode=::NewEnumerator4
    Position Offset Coordinate Space=::Local
    UsePositionOffset=true
    Write Material Random=true
   )
   &GridLocation(
    Coordinate Space:ENiagaraCoordinateSpace=::World
    Module.Jitter Inside Grid Cells=false
    Module.Override Random Seed=false
    Normalize Offsets=false
    XYZ Dimensions:Vector3f=&MakeVector(
     X=&Divide_Float(A=$Emitter.VolumeRadius, B=$Emitter.PaddingPerCell)
     Y=&Divide_Float(A=$Emitter.VolumeRadius, B=$Emitter.PaddingPerCell)
     Z=0
    )
   )
   &InitialMeshOrientation(
    Mesh Orientation Mode=::NewEnumerator2
    Rotation=&SelectVectorFromArray(
     Array Sampling Mode=::NewEnumerator0
     Direct Array Index=?
     Vector Selection Array=DataInterface() {propertyValues {FloatData [<0 0 0>, <0 0 0.25>, <0 0 0.5>]}}
     Random Seed=0
    )
    Use Rotation Vector=true
   )
  ]
  particleUpdateScript [
   &ParticleState(Kill Particles When Lifetime Has Elapsed=false, Let Infinitely Lived Particles Die When Emitter Deactivates=true)
   &ScaleColor(
    Scale Alpha=&FN_FrontEndFloatModifier(Gameplay=1, Front End=1)
    Scale RGB=&Multiply_VectorByFloat(Vector=<1 1 1>, Float=&FN_FrontEndFloatModifier(Gameplay=1, Front End=1))
    ScaleA=true
    ScaleRGB=true
   ) {module=ScaleColor001}
   &SkeletalMeshLocation(
    Bone Sampling Mode=::NewEnumerator4
    MaskSpawnGroup=false
    Orientation Sampling=::NewEnumerator1
    Position Sampling=::NewEnumerator2
    Sampled Orientation Coordinate Space=::World
    Sampled Position Coordinate Space=::World
    Sampled Position Offset Coordinate Space:ENiagaraSkelMeshTransforms=::NewEnumerator0
    Sampled Scale Coordinate Space:ENiagaraCoordinateSpace=::Simulation
    Skeletal Mesh=DataInterface() {propertyValues {
      DefaultMesh=null
      ExcludeBoneName=root
      OptionalComponentClass=null
      PreviewMesh=&SK_M_SML_Ninja_04_Body
      SkinningMode=SkinOnTheFly
      SoftSourceActor=null
      SourceMode=AttachParent
      UvSetIndex=0
      WholeMeshLOD=-1
      bExcludeBone=false
      bReadDeformedGeometry=true
      bRequireCurrentFrameData=true
      ComponentTags []
      FilteredBones [pelvis,neck_02]
      FilteredSockets []
      MeshUserParameter {parameter {name=None, typeDefHandle {registeredTypeIndex=1}, varData []}}
      SamplingRegions []
     }}
   )
   /SetVariables(Particles.PelvisPosition:NiagaraPosition=$Output.SkeletalMeshLocation.SampledPositionWithoutOffset) {isSetParameters=true, module=SetVariables_#3}
   &SkeletalMeshLocation(
    Bone / Socket Index=1
    Bone Sampling Mode=::NewEnumerator4
    MaskSpawnGroup=false
    Orientation Sampling=::NewEnumerator1
    Position Sampling=::NewEnumerator2
    Sampled Position Coordinate Space=::World
    Sampled Position Offset Coordinate Space:ENiagaraSkelMeshTransforms=::NewEnumerator0
    Sampled Scale Coordinate Space:ENiagaraCoordinateSpace=::Simulation
    Skeletal Mesh=DataInterface() {propertyValues {
      DefaultMesh=null
      ExcludeBoneName=root
      OptionalComponentClass=null
      PreviewMesh=&SK_M_SML_Ninja_04_Body
      SkinningMode=SkinOnTheFly
      SoftSourceActor=null
      SourceMode=AttachParent
      UvSetIndex=0
      WholeMeshLOD=-1
      bExcludeBone=false
      bReadDeformedGeometry=true
      bRequireCurrentFrameData=false
      ComponentTags []
      FilteredBones [pelvis,neck_02]
      FilteredSockets []
      MeshUserParameter {parameter {name=None, typeDefHandle {registeredTypeIndex=1}, varData []}}
      SamplingRegions []
     }}
   ) {module=SkeletalMeshLocation001}
   /SetVariables(Particles.NeckPosition:NiagaraPosition=$Output.SkeletalMeshLocation001.SampledPositionWithoutOffset) {isSetParameters=true, module=SetVariables_#4}
   &SkeletalMeshLocation(
    Bone / Socket Index=2
    Bone Sampling Mode=::NewEnumerator4
    MaskSpawnGroup=false
    Orientation Sampling=::NewEnumerator1
    Position Sampling=::NewEnumerator2
    Sampled Position Offset Coordinate Space:ENiagaraSkelMeshTransforms=::NewEnumerator0
    Sampled Scale Coordinate Space:ENiagaraCoordinateSpace=::Simulation
    Skeletal Mesh=DataInterface() {propertyValues {
      DefaultMesh=null
      ExcludeBoneName=root
      OptionalComponentClass=null
      PreviewMesh=&SK_M_SML_Ninja_04_Body
      SkinningMode=SkinOnTheFly
      SoftSourceActor=null
      SourceMode=AttachParent
      UvSetIndex=0
      WholeMeshLOD=-1
      bExcludeBone=false
      bReadDeformedGeometry=true
      bRequireCurrentFrameData=true
      ComponentTags []
      FilteredBones [pelvis,neck_02,foot_l,ik_foot_l]
      FilteredSockets []
      MeshUserParameter {parameter {name=None, typeDefHandle {registeredTypeIndex=1}, varData []}}
      SamplingRegions []
     }}
   ) {module=SkeletalMeshLocation002}
   /SetVariables(Particles.LeftLegPosition:NiagaraPosition=$Output.SkeletalMeshLocation002.SampledPositionWithoutOffset) {isSetParameters=true, module=SetVariables_#5}
   &SkeletalMeshLocation(
    Bone / Socket Index=3
    Bone Sampling Mode=::NewEnumerator4
    MaskSpawnGroup=false
    Orientation Sampling=::NewEnumerator1
    Position Sampling=::NewEnumerator2
    Sampled Position Offset Coordinate Space:ENiagaraSkelMeshTransforms=::NewEnumerator0
    Sampled Scale Coordinate Space:ENiagaraCoordinateSpace=::Simulation
    Skeletal Mesh=DataInterface() {propertyValues {
      DefaultMesh=null
      ExcludeBoneName=root
      OptionalComponentClass=null
      PreviewMesh=&SK_M_SML_Ninja_04_Body
      SkinningMode=SkinOnTheFly
      SoftSourceActor=null
      SourceMode=AttachParent
      UvSetIndex=0
      WholeMeshLOD=-1
      bExcludeBone=false
      bReadDeformedGeometry=true
      bRequireCurrentFrameData=true
      ComponentTags []
      FilteredBones [pelvis,neck_02,calf_r,ik_foot_r]
      FilteredSockets []
      MeshUserParameter {parameter {name=None, typeDefHandle {registeredTypeIndex=1}, varData []}}
      SamplingRegions []
     }}
   ) {module=SkeletalMeshLocation003}
   /SetVariables(Particles.RightLegPosition:NiagaraPosition=$Output.SkeletalMeshLocation003.SampledPositionWithoutOffset) {isSetParameters=true, module=SetVariables_#6}
   &SkeletalMeshLocation(
    Bone Sampling Mode=::NewEnumerator4
    MaskSpawnGroup=false
    Orientation Sampling=::NewEnumerator1
    Position Sampling=::NewEnumerator2
    Sampled Position Offset Coordinate Space:ENiagaraSkelMeshTransforms=::NewEnumerator0
    Sampled Scale Coordinate Space:ENiagaraCoordinateSpace=::Simulation
    Skeletal Mesh=DataInterface() {propertyValues {
      DefaultMesh=null
      ExcludeBoneName=root
      OptionalComponentClass=null
      PreviewMesh=&SK_M_SML_Ninja_04_Body
      SkinningMode=SkinOnTheFly
      SoftSourceActor=null
      SourceMode=AttachParent
      UvSetIndex=0
      WholeMeshLOD=-1
      bExcludeBone=false
      bReadDeformedGeometry=true
      bRequireCurrentFrameData=true
      ComponentTags []
      FilteredBones [calf_l]
      FilteredSockets []
      MeshUserParameter {parameter {name=None, typeDefHandle {registeredTypeIndex=1}, varData []}}
      SamplingRegions []
     }}
   ) {module=SkeletalMeshLocation004}
   /SetVariables(Particles.LeftKneePosition:NiagaraPosition=$Particles.SkeletalMeshLocation004.SampledPosition) {isSetParameters=true, module=SetVariables_#7}
   &SkeletalMeshLocation(
    Bone Sampling Mode=::NewEnumerator4
    MaskSpawnGroup=false
    Orientation Sampling=::NewEnumerator1
    Position Sampling=::NewEnumerator2
    Sampled Position Offset Coordinate Space:ENiagaraSkelMeshTransforms=::NewEnumerator0
    Sampled Scale Coordinate Space:ENiagaraCoordinateSpace=::Simulation
    Skeletal Mesh=DataInterface() {propertyValues {
      DefaultMesh=null
      ExcludeBoneName=root
      OptionalComponentClass=null
      PreviewMesh=&SK_M_SML_Ninja_04_Body
      SkinningMode=SkinOnTheFly
      SoftSourceActor=null
      SourceMode=AttachParent
      UvSetIndex=0
      WholeMeshLOD=-1
      bExcludeBone=false
      bReadDeformedGeometry=true
      bRequireCurrentFrameData=true
      ComponentTags []
      FilteredBones [calf_r]
      FilteredSockets []
      MeshUserParameter {parameter {name=None, typeDefHandle {registeredTypeIndex=1}, varData []}}
      SamplingRegions []
     }}
   ) {module=SkeletalMeshLocation005}
   /SetVariables(Particles.RightKneePosition:NiagaraPosition=$Output.SkeletalMeshLocation005.SampledPositionWithoutOffset) {isSetParameters=true, module=SetVariables_#8}
   &SolveForcesAndVelocity(Clamp Velocity=false, Limit Acceleration=false)
   /SetVariables(Particles.PositionCenter:NiagaraPosition=&AddVectorToPosition(Position=$Engine.Owner.Position, Vector=$Emitter.OriginOffset)) {isSetParameters=true, module=SetVariables_#9}
   &self:TreadmillParticles_01(
    GridSize:NiagaraFloat=&Multiply_Float(
     A=$Emitter.VolumeRadius
     B=&Lerp_Float(
      A=2.5
      B=2.5
      Alpha=&MakeCustomFloatFromBool(
       Bool=&SetBoolByFloatComparison(A=&VectorLength(Vector=$User.OwnerLocalVelocity), B=160)
       True Float=2
       False Float=0
      )
     )
    )
    InputPosition:NiagaraPosition=$Particles.PositionCenter
   ) {module=TreadmillParticles_01}
   /SetVariables(Particles.Position=&MakeVector(
     X=&MakeFloatFromVector(Vector=Linked() {linkedVariable {name=Particles.Position, type=NiagaraPosition}}, Channel=::NewEnumerator0)
     Y=&MakeFloatFromVector(Vector=Linked() {linkedVariable {name=Particles.Position, type=NiagaraPosition}}, Channel=::NewEnumerator1)
     Z=&MakeFloatFromVector(Vector=&SimulationPosition(), Channel=::NewEnumerator2)
    )) {isSetParameters=true, module=SetVariables_#14}
   /SetVariables(Particles.Position=&AddVectorToPosition(Position=$Particles.Position, Vector=<0 0 100>)) {isSetParameters=true, module=SetVariables_#15}
   /SetVariables(Particles.Position=&AddVectorToPosition(
     Position=$Particles.Position
     Vector=&FN_FrontEndVector3Modifier(
      Gameplay=<0 0 0>
      Front End=&RotateVectorByQuaternion(
       Vector To Rotate=&TransformVector(Vector=&Multiply_VectorByFloat(Vector=<0 -50 0>, Float=$Engine.DeltaTime), Destination Space=::World)
       Quaternion=$Output.SocketLocation.SampledRotation
      )
     )
    )) {isSetParameters=true, module=SetVariables_#16}
   /SetVariables(Particles.LerpedAudioReactive:NiagaraFloat=&self:Lerp_Audio_Reactive_EQ(
     Default:Vector3f=&VectorFromFloat(Value=$NPC.FortniteNPC.EventAudio_Amplitude)
     ChangeRate:NiagaraFloat=15
    )) {isSetParameters=true, module=SetVariables_#10}
   &self:GridCollision(TraceDir:Vector3f=<0 0 -1>, TraceDist=250, bIsFrontEnd:NiagaraBool=$User.bIsFrontEndPreview) {module=GridCollision}
   /SetVariables(Particles.TargetPos:Vector3f=&self:Return_Closest_Joint(
     PelvisPosition:Vector3f=$Particles.PelvisPosition
     NeckPosition:Vector3f=$Particles.NeckPosition
     LeftLegPosition:Vector3f=$Particles.LeftLegPosition
     RightLegPosition:Vector3f=$Particles.RightLegPosition
     LeftCalfPosition:Vector3f=$Particles.LeftKneePosition
     RightCalfPosition:Vector3f=$Particles.RightKneePosition
     NeckOffset:Vector3f=&Lerp_Vector(
      A=<0 0 -60>
      B=<0 0 -20>
      Alpha=&RemapRange(
       Input Value To Remap=&ClampFloat(Float=$Engine.System.Age, Max=0.8)
       Input Min=0
       Input Max=0.8
       Output Min=0
       Output Max=1
      )
     )
     PelvisOffset:Vector3f=&Lerp_Vector(
      A=<0 0 -60>
      B=<0 0 -10>
      Alpha=&RemapRange(
       Input Value To Remap=&ClampFloat(Float=$Engine.System.Age, Max=0.8)
       Input Min=0
       Input Max=0.8
       Output Min=0
       Output Max=1
      )
     )
     KneeOffset:Vector3f=<0 0 -20>
     FootOffset:Vector3f=<0 0 -20>
     ParticlePosition:Vector3f=Linked() {linkedVariable {name=Particles.Position, type=NiagaraPosition}}
    )) {isSetParameters=true, module=SetVariables_#11}
   &RegionMask(
    Invert Region=true
    Offset:Vector3f=<0 165 0>
    Offset Space:ENiagaraCoordinateSpace=::Local
    Outer Float=0
    Region Dimensions:Vector3f=<200 850 1500>
    Region Falloff=200
    Region Type:ENiagaraDebugDrawShape=::NewEnumerator1
    Rotation:Vector3f=$Engine.Owner.SystemYAxis
    Rotation Angle Type:ENiagara_AngleInput=::NewEnumerator0
    UseOffset=true
    UseRotation=true
   )
   &Timeline(
    Additional Output:ENiagara_TimelineOutput=::NewEnumerator1
    Additional Output Mode:ENiagara_TimelineOutputMode=::NewEnumerator1
    Float Curve=DataInterface() {propertyValues {Curve {
       defaultValue=0
       keys [
        {interpMode=RCIM_Cubic, tangentMode=RCTM_User, time=0, value=0}
        {arriveTangent=1, interpMode=RCIM_Cubic, leaveTangent=1, time=0.2, value=0}
        {interpMode=RCIM_Cubic, tangentMode=RCTM_User, time=1, value=1}
       ]
      }}}
    Play=&Boolean_NotOperation(Bool=$Particles.RegionMask.IsInsideMask)
    Play Rate=&Lerp_Float(
     A=5
     B=20
     Alpha=&RemapRange(
      Input Value To Remap=&VectorLength(Vector=$User.OwnerLocalVelocity)
      Input Min=0
      Input Max=800
      Output Min=0
      Output Max=1
     )
    )
    Rewind Rate:NiagaraFloat=&Lerp_Float(
     A=5
     B=20
     Alpha=&RemapRange(
      Input Value To Remap=&VectorLength(Vector=$User.OwnerLocalVelocity)
      Input Min=0
      Input Max=800
      Output Min=0
      Output Max=1
     )
    )
   )
   &ScaleMeshSize(Scale Factor=&VectorFromFloat(Value=&Multiply_Float(
      A=$Particles.Timeline.Float
      B=&FloatFromCurve(FloatCurve=DataInterface() {propertyValues {Curve {
          defaultValue=1
          keys [{interpMode=RCIM_Cubic, tangentMode=RCTM_User, time=0, value=1}, {interpMode=RCIM_Cubic, tangentMode=RCTM_User, time=1, value=1}]
         }}})
     ))) {module=ScaleMeshSize001}
   /SetVariables(Particles.ParticleZScaleFactor:NiagaraFloat=&ClampFloat(
     Float=&Multiply_Float(
      A=&self:Lerp_Player_Position_EQ(
       TargetPos:Vector3f=$Particles.TargetPos
       ChangeRate:NiagaraFloat=20
       XY Distance:NiagaraFloat=100
       ParticlePosition:Vector3f=Linked() {linkedVariable {name=Particles.CollisionPos, type=NiagaraPosition}}
       BaseScale:NiagaraFloat=&Lerp_Float(
        A=-200
        B=125
        Alpha=&RemapRange(Input Value To Remap=$Particles.LerpedAudioReactive, Input Min=0.3, Input Max=1, Output Min=0, Output Max=1)
       )
       InputMinTest:NiagaraFloat=1
       TargetZOffset:NiagaraFloat=0
       DeltaTime:NiagaraFloat=&Multiply_Float(A=$Engine.DeltaTime, B=1)
      )
      B=&Lerp_Float(
       A=0.1
       B=0.12
       Alpha=&RemapRange(
        Input Value To Remap=&VectorLength(Vector=$User.OwnerLocalVelocity)
        Input Min=0
        Input Max=800
        Output Min=0
        Output Max=1
       )
      )
     )
     Min=1
     Max=&Lerp_Float(
      A=0
      B=35
      Alpha=&RemapRange(
       Input Value To Remap=&ClampFloat(Float=$Engine.System.Age, Max=0.5)
       Input Min=0
       Input Max=0.5
       Output Min=0
       Output Max=1
      )
     )
    )) {isSetParameters=true, module=SetVariables_#12}
   &ScaleMeshSize(Scale Factor=&MakeVector(X=1, Y=1, Z=$Particles.ParticleZScaleFactor)) {module=ScaleMeshSize002}
   &ScaleMeshSize(Scale Factor=&Lerp_Vector(
     A=<1 1 1>
     B=<1 1 0.5>
     Alpha=&RemapRange(
      Input Value To Remap=&self:XY_Distance_From_Closest_Joint(Target:Vector3f=$Particles.TargetPos, Current:Vector3f=Linked() {linkedVariable {name=Particles.Position, type=NiagaraPosition}})
      Input Min=0
      Input Max=400
      Output Min=0
      Output Max=1
     )
    )) {module=ScaleMeshSize008}
   /SetVariables(Particles.AudioReactiveColorValue:NiagaraFloat=&Multiply_Float(
     A=&Lerp_Float(
      A=1
      B=0.5
      Alpha=&RemapRange(
       Input Value To Remap=&self:XY_Distance_From_Closest_Joint(Target:Vector3f=$Particles.TargetPos, Current:Vector3f=Linked() {linkedVariable {name=Particles.Position, type=NiagaraPosition}})
       Input Min=0
       Input Max=400
       Output Min=0
       Output Max=1
      )
     )
     B=&Lerp_Float(
      A=-0.8
      B=1
      Alpha=&RemapRange(Input Value To Remap=$Particles.LerpedAudioReactive, Input Min=0.3, Input Max=1, Output Min=0, Output Max=1)
     )
    )) {isSetParameters=true, module=SetVariables_#13}
   &ScaleColor(
    Scale Alpha=&Lerp_Float(
     A=1
     B=0
     Alpha=&RemapRange(
      Input Value To Remap=&DistanceBetweenPositions(A=$Particles.Position, B=Linked() {linkedVariable {name=Particles.PelvisPosition, type=Vector3f}})
      Input Min=70
      Input Max=160
      Output Min=0
      Output Max=1
     )
    )
    Scale RGB=&Lerp_Vector(
     A=&Lerp_Vector(
      A=&FN_FrontEndVector3Modifier(
       Gameplay=&MakeVectorFromLinearColorRGB(Color=<rgba 0.1048 5 1.9461 1>)
       Front End=&MakeVectorFromLinearColorRGB(Color=<rgba 0 1.5576 3 1>)
      )
      B=&MakeVectorFromLinearColorRGB(Color=<rgba 0 2.5179 5 1>)
      Alpha=&RemapRange(
       Input Value To Remap=&Sine(
        Normalized Angle=&Add_Float(A=$Emitter.Age, B=&RandomRangeFloat(Minimum=-1, Maximum=1))
        Period=5
        Scale=1
        Bias=0
       )
       Input Min=-1
       Input Max=1
       Output Min=0
       Output Max=0.6
      )
     )
     B=&MakeVectorFromLinearColorRGB(Color=<rgba 0.8642 0.7917 1 1>)
     Alpha=&RemapRange(
      Input Value To Remap=&DistanceBetweenPositions(
       A=&MakeVector(
        X=&MakeFloatFromVector(Vector=Linked() {linkedVariable {name=Particles.Position, type=NiagaraPosition}}, Channel=::NewEnumerator0)
        Y=&MakeFloatFromVector(Vector=Linked() {linkedVariable {name=Particles.Position, type=NiagaraPosition}}, Channel=::NewEnumerator1)
        Z=0
       )
       B=&MakeVector(
        X=&MakeFloatFromVector(Vector=&SimulationPosition(), Channel=::NewEnumerator0)
        Y=&MakeFloatFromVector(Vector=&SimulationPosition(), Channel=::NewEnumerator1)
        Z=0
       )
      )
      Input Min=0
      Input Max=&Lerp_Float(
       A=50
       B=120
       Alpha=&RemapRange(Input Value To Remap=$Particles.LerpedAudioReactive, Input Min=0.6, Input Max=1, Output Min=0, Output Max=1)
      )
      Output Min=0
      Output Max=1
     )
    )
    ScaleA=true
    ScaleRGB=true
   ) {module=ScaleColor002}
   &ScaleColor(
    Scale Alpha=&Lerp_Float(
     A=1
     B=0
     Alpha=&RemapRange(
      Input Value To Remap=&DistanceBetweenPositions(A=$Particles.Position, B=Linked() {linkedVariable {name=Particles.PelvisPosition, type=Vector3f}})
      Input Min=70
      Input Max=160
      Output Min=0
      Output Max=1
     )
    )
    Scale RGB=&Lerp_Vector(
     A=<1 1 1>
     B=&MakeVectorFromLinearColorRGB(Color=<rgba 0.0968 4.6189 1.7978 1>)
     Alpha=&ClampFloat(Float=&Multiply_Float(A=$Particles.AudioReactiveColorValue, B=1))
    )
    ScaleA=true
    ScaleRGB=true
   )
   &ScaleMeshSize(Scale Factor=&Lerp_Vector(
     A=<1 1 1>
     B=&Lerp_Vector(
      A=<1 1 0.8>
      B=<1 1 1.1>
      Alpha=&Lerp_Float(
       A=0
       B=&Add_Float(
        A=&Sine(
         Normalized Angle=&Add_Float(
          A=$Emitter.Age
          B=&Multiply_Float(
           A=&DistanceBetweenPositions(A=$Particles.Position, B=Linked() {linkedVariable {name=Particles.PelvisPosition, type=Vector3f}})
           B=-0.01
          )
         )
         Period=1
         Scale=1
         Bias=0
        )
        B=1
       )
       Alpha=&RemapRange(
        Input Value To Remap=&DistanceBetweenPositions(A=Linked() {linkedVariable {name=Particles.PelvisPosition, type=Vector3f}}, B=$Particles.Position)
        Input Min=0
        Input Max=150
        Output Min=0
        Output Max=1
       )
      )
     )
     Alpha=&self:Height_Gate(
      ChangeRate:NiagaraFloat=12
      Max:NiagaraFloat=0.45
      Particle Z Scale:NiagaraFloat=&MakeFloatFromVector(Vector=$Particles.Scale, Channel=::NewEnumerator2)
     )
    ))
  ]
  rendererClasses [&NiagaraMeshRendererProperties]
  renderers [{rendererClass=&NiagaraMeshRendererProperties, rendererIndex=0}]
 }
 Visualizer_Low_Med {
  bEnabled=false
  emitterUpdateScript [
   &EmitterState(
    Inactive Response=::NewEnumerator1
    Life Cycle Mode=::NewEnumerator1
    Loop Behavior=::NewEnumerator1
    Loop Duration Mode=::NewEnumerator1
    UseLoopDelay=false
   )
   /SetVariables(
    Emitter.OriginOffset:Vector3f=&TransformVector(Vector=<0 0 0>, Destination Space=::World)
    Emitter.PaddingPerCell=4
    Emitter.VolumeRadius=110
   ) {isSetParameters=true, module=SetVariables_#2}
   &SpawnParticlesInGrid(X Count=11, Y Count=11, Z Count=0)
  ]
  particleSpawnScript [
   &InitializeParticle(
    Color=&Multiply_LinearColorByFloat(Linear Color=<rgba 1 1 1 0.08>, Float=&FN_FrontEndFloatModifier(Gameplay=1, Front End=1))
    Mesh Scale=<0.27 0.27 0.1>
    Mesh Scale Mode=::NewEnumerator4
    Position Offset Coordinate Space=::Local
    UsePositionOffset=true
    Write Material Random=true
   )
   &GridLocation(
    Coordinate Space:ENiagaraCoordinateSpace=::World
    Module.Jitter Inside Grid Cells=false
    Module.Override Random Seed=false
    Normalize Offsets=false
    XYZ Dimensions:Vector3f=&MakeVector(
     X=&Divide_Float(A=$Emitter.VolumeRadius, B=$Emitter.PaddingPerCell)
     Y=&Divide_Float(A=$Emitter.VolumeRadius, B=$Emitter.PaddingPerCell)
     Z=0
    )
   )
   &InitialMeshOrientation(
    Mesh Orientation Mode=::NewEnumerator2
    Rotation=&SelectVectorFromArray(
     Array Sampling Mode=::NewEnumerator0
     Direct Array Index=?
     Vector Selection Array=DataInterface() {propertyValues {FloatData [<0 0 0>, <0 0 0.25>, <0 0 0.5>]}}
     Random Seed=0
    )
    Use Rotation Vector=true
   )
  ]
  particleUpdateScript [
   &ParticleState(Kill Particles When Lifetime Has Elapsed=false, Let Infinitely Lived Particles Die When Emitter Deactivates=true)
   &ScaleColor(
    Scale Alpha=&FN_FrontEndFloatModifier(Gameplay=1, Front End=1)
    Scale RGB=&Multiply_VectorByFloat(Vector=<1 1 1>, Float=&FN_FrontEndFloatModifier(Gameplay=1, Front End=1))
    ScaleA=true
    ScaleRGB=true
   ) {module=ScaleColor001}
   &SkeletalMeshLocation(
    Bone Sampling Mode=::NewEnumerator4
    MaskSpawnGroup=false
    Orientation Sampling=::NewEnumerator1
    Position Sampling=::NewEnumerator2
    Sampled Orientation Coordinate Space=::World
    Sampled Position Coordinate Space=::World
    Sampled Position Offset Coordinate Space:ENiagaraSkelMeshTransforms=::NewEnumerator0
    Sampled Scale Coordinate Space:ENiagaraCoordinateSpace=::Simulation
    Skeletal Mesh=DataInterface() {propertyValues {
      DefaultMesh=null
      ExcludeBoneName=root
      OptionalComponentClass=null
      PreviewMesh=&SK_M_SML_Ninja_04_Body
      SkinningMode=SkinOnTheFly
      SoftSourceActor=null
      SourceMode=AttachParent
      UvSetIndex=0
      WholeMeshLOD=-1
      bExcludeBone=false
      bReadDeformedGeometry=true
      bRequireCurrentFrameData=true
      ComponentTags []
      FilteredBones [pelvis,neck_02]
      FilteredSockets []
      MeshUserParameter {parameter {name=None, typeDefHandle {registeredTypeIndex=1}, varData []}}
      SamplingRegions []
     }}
   )
   /SetVariables(Particles.PelvisPosition:NiagaraPosition=$Output.SkeletalMeshLocation.SampledPositionWithoutOffset) {isSetParameters=true, module=SetVariables_#3}
   &SkeletalMeshLocation(
    Bone / Socket Index=1
    Bone Sampling Mode=::NewEnumerator4
    MaskSpawnGroup=false
    Orientation Sampling=::NewEnumerator1
    Position Sampling=::NewEnumerator2
    Sampled Position Coordinate Space=::World
    Sampled Position Offset Coordinate Space:ENiagaraSkelMeshTransforms=::NewEnumerator0
    Sampled Scale Coordinate Space:ENiagaraCoordinateSpace=::Simulation
    Skeletal Mesh=DataInterface() {propertyValues {
      DefaultMesh=null
      ExcludeBoneName=root
      OptionalComponentClass=null
      PreviewMesh=&SK_M_SML_Ninja_04_Body
      SkinningMode=SkinOnTheFly
      SoftSourceActor=null
      SourceMode=AttachParent
      UvSetIndex=0
      WholeMeshLOD=-1
      bExcludeBone=false
      bReadDeformedGeometry=true
      bRequireCurrentFrameData=false
      ComponentTags []
      FilteredBones [pelvis,neck_02]
      FilteredSockets []
      MeshUserParameter {parameter {name=None, typeDefHandle {registeredTypeIndex=1}, varData []}}
      SamplingRegions []
     }}
   ) {module=SkeletalMeshLocation001}
   /SetVariables(Particles.NeckPosition:NiagaraPosition=$Output.SkeletalMeshLocation001.SampledPositionWithoutOffset) {isSetParameters=true, module=SetVariables_#4}
   &SkeletalMeshLocation(
    Bone / Socket Index=2
    Bone Sampling Mode=::NewEnumerator4
    MaskSpawnGroup=false
    Orientation Sampling=::NewEnumerator1
    Position Sampling=::NewEnumerator2
    Sampled Position Offset Coordinate Space:ENiagaraSkelMeshTransforms=::NewEnumerator0
    Sampled Scale Coordinate Space:ENiagaraCoordinateSpace=::Simulation
    Skeletal Mesh=DataInterface() {propertyValues {
      DefaultMesh=null
      ExcludeBoneName=root
      OptionalComponentClass=null
      PreviewMesh=&SK_M_SML_Ninja_04_Body
      SkinningMode=SkinOnTheFly
      SoftSourceActor=null
      SourceMode=AttachParent
      UvSetIndex=0
      WholeMeshLOD=-1
      bExcludeBone=false
      bReadDeformedGeometry=true
      bRequireCurrentFrameData=true
      ComponentTags []
      FilteredBones [pelvis,neck_02,foot_l,ik_foot_l]
      FilteredSockets []
      MeshUserParameter {parameter {name=None, typeDefHandle {registeredTypeIndex=1}, varData []}}
      SamplingRegions []
     }}
   ) {module=SkeletalMeshLocation002}
   /SetVariables(Particles.LeftLegPosition:NiagaraPosition=$Output.SkeletalMeshLocation002.SampledPositionWithoutOffset) {isSetParameters=true, module=SetVariables_#5}
   &SkeletalMeshLocation(
    Bone / Socket Index=3
    Bone Sampling Mode=::NewEnumerator4
    MaskSpawnGroup=false
    Orientation Sampling=::NewEnumerator1
    Position Sampling=::NewEnumerator2
    Sampled Position Offset Coordinate Space:ENiagaraSkelMeshTransforms=::NewEnumerator0
    Sampled Scale Coordinate Space:ENiagaraCoordinateSpace=::Simulation
    Skeletal Mesh=DataInterface() {propertyValues {
      DefaultMesh=null
      ExcludeBoneName=root
      OptionalComponentClass=null
      PreviewMesh=&SK_M_SML_Ninja_04_Body
      SkinningMode=SkinOnTheFly
      SoftSourceActor=null
      SourceMode=AttachParent
      UvSetIndex=0
      WholeMeshLOD=-1
      bExcludeBone=false
      bReadDeformedGeometry=true
      bRequireCurrentFrameData=true
      ComponentTags []
      FilteredBones [pelvis,neck_02,calf_r,ik_foot_r]
      FilteredSockets []
      MeshUserParameter {parameter {name=None, typeDefHandle {registeredTypeIndex=1}, varData []}}
      SamplingRegions []
     }}
   ) {module=SkeletalMeshLocation003}
   /SetVariables(Particles.RightLegPosition:NiagaraPosition=$Output.SkeletalMeshLocation003.SampledPositionWithoutOffset) {isSetParameters=true, module=SetVariables_#6}
   &SkeletalMeshLocation(
    Bone Sampling Mode=::NewEnumerator4
    MaskSpawnGroup=false
    Orientation Sampling=::NewEnumerator1
    Position Sampling=::NewEnumerator2
    Sampled Position Offset Coordinate Space:ENiagaraSkelMeshTransforms=::NewEnumerator0
    Sampled Scale Coordinate Space:ENiagaraCoordinateSpace=::Simulation
    Skeletal Mesh=DataInterface() {propertyValues {
      DefaultMesh=null
      ExcludeBoneName=root
      OptionalComponentClass=null
      PreviewMesh=&SK_M_SML_Ninja_04_Body
      SkinningMode=SkinOnTheFly
      SoftSourceActor=null
      SourceMode=AttachParent
      UvSetIndex=0
      WholeMeshLOD=-1
      bExcludeBone=false
      bReadDeformedGeometry=true
      bRequireCurrentFrameData=true
      ComponentTags []
      FilteredBones [calf_l]
      FilteredSockets []
      MeshUserParameter {parameter {name=None, typeDefHandle {registeredTypeIndex=1}, varData []}}
      SamplingRegions []
     }}
   ) {module=SkeletalMeshLocation004}
   /SetVariables(Particles.LeftKneePosition:NiagaraPosition=$Particles.SkeletalMeshLocation004.SampledPosition) {isSetParameters=true, module=SetVariables_#7}
   &SkeletalMeshLocation(
    Bone Sampling Mode=::NewEnumerator4
    MaskSpawnGroup=false
    Orientation Sampling=::NewEnumerator1
    Position Sampling=::NewEnumerator2
    Sampled Position Offset Coordinate Space:ENiagaraSkelMeshTransforms=::NewEnumerator0
    Sampled Scale Coordinate Space:ENiagaraCoordinateSpace=::Simulation
    Skeletal Mesh=DataInterface() {propertyValues {
      DefaultMesh=null
      ExcludeBoneName=root
      OptionalComponentClass=null
      PreviewMesh=&SK_M_SML_Ninja_04_Body
      SkinningMode=SkinOnTheFly
      SoftSourceActor=null
      SourceMode=AttachParent
      UvSetIndex=0
      WholeMeshLOD=-1
      bExcludeBone=false
      bReadDeformedGeometry=true
      bRequireCurrentFrameData=true
      ComponentTags []
      FilteredBones [calf_r]
      FilteredSockets []
      MeshUserParameter {parameter {name=None, typeDefHandle {registeredTypeIndex=1}, varData []}}
      SamplingRegions []
     }}
   ) {module=SkeletalMeshLocation005}
   /SetVariables(Particles.RightKneePosition:NiagaraPosition=$Output.SkeletalMeshLocation005.SampledPositionWithoutOffset) {isSetParameters=true, module=SetVariables_#8}
   &SolveForcesAndVelocity(Clamp Velocity=false, Limit Acceleration=false)
   /SetVariables(Particles.PositionCenter:NiagaraPosition=&AddVectorToPosition(Position=$Engine.Owner.Position, Vector=$Emitter.OriginOffset)) {isSetParameters=true, module=SetVariables_#9}
   &self:TreadmillParticles_01(
    GridSize:NiagaraFloat=&Multiply_Float(
     A=$Emitter.VolumeRadius
     B=&Lerp_Float(
      A=2.5
      B=2.5
      Alpha=&MakeCustomFloatFromBool(
       Bool=&SetBoolByFloatComparison(A=&VectorLength(Vector=$User.OwnerLocalVelocity), B=160)
       True Float=2
       False Float=0
      )
     )
    )
    InputPosition:NiagaraPosition=$Particles.PositionCenter
   ) {module=TreadmillParticles_01}
   /SetVariables(Particles.Position=&MakeVector(
     X=&MakeFloatFromVector(Vector=Linked() {linkedVariable {name=Particles.Position, type=NiagaraPosition}}, Channel=::NewEnumerator0)
     Y=&MakeFloatFromVector(Vector=Linked() {linkedVariable {name=Particles.Position, type=NiagaraPosition}}, Channel=::NewEnumerator1)
     Z=&MakeFloatFromVector(Vector=&SimulationPosition(), Channel=::NewEnumerator2)
    )) {isSetParameters=true, module=SetVariables_#14}
   /SetVariables(Particles.Position=&AddVectorToPosition(Position=$Particles.Position, Vector=<0 0 100>)) {isSetParameters=true, module=SetVariables_#15}
   /SetVariables(Particles.Position=&AddVectorToPosition(
     Position=$Particles.Position
     Vector=&FN_FrontEndVector3Modifier(
      Gameplay=<0 0 0>
      Front End=&RotateVectorByQuaternion(
       Vector To Rotate=&TransformVector(Vector=&Multiply_VectorByFloat(Vector=<0 -50 0>, Float=$Engine.DeltaTime), Destination Space=::World)
       Quaternion=$Output.SocketLocation.SampledRotation
      )
     )
    )) {isSetParameters=true, module=SetVariables_#16}
   /SetVariables(Particles.LerpedAudioReactive:NiagaraFloat=&self:Lerp_Audio_Reactive_EQ(
     Default:Vector3f=&VectorFromFloat(Value=$NPC.FortniteNPC.EventAudio_Amplitude)
     ChangeRate:NiagaraFloat=15
    )) {isSetParameters=true, module=SetVariables_#10}
   &self:GridCollision(TraceDir:Vector3f=<0 0 -1>, TraceDist=250, bIsFrontEnd:NiagaraBool=$User.bIsFrontEndPreview) {module=GridCollision}
   /SetVariables(Particles.TargetPos:Vector3f=&self:Return_Closest_Joint(
     PelvisPosition:Vector3f=$Particles.PelvisPosition
     NeckPosition:Vector3f=$Particles.NeckPosition
     LeftLegPosition:Vector3f=$Particles.LeftLegPosition
     RightLegPosition:Vector3f=$Particles.RightLegPosition
     LeftCalfPosition:Vector3f=$Particles.LeftKneePosition
     RightCalfPosition:Vector3f=$Particles.RightKneePosition
     NeckOffset:Vector3f=&Lerp_Vector(
      A=<0 0 -60>
      B=<0 0 -20>
      Alpha=&RemapRange(
       Input Value To Remap=&ClampFloat(Float=$Engine.System.Age, Max=0.8)
       Input Min=0
       Input Max=0.8
       Output Min=0
       Output Max=1
      )
     )
     PelvisOffset:Vector3f=&Lerp_Vector(
      A=<0 0 -60>
      B=<0 0 -10>
      Alpha=&RemapRange(
       Input Value To Remap=&ClampFloat(Float=$Engine.System.Age, Max=0.8)
       Input Min=0
       Input Max=0.8
       Output Min=0
       Output Max=1
      )
     )
     KneeOffset:Vector3f=<0 0 -20>
     FootOffset:Vector3f=<0 0 -20>
     ParticlePosition:Vector3f=Linked() {linkedVariable {name=Particles.Position, type=NiagaraPosition}}
    )) {isSetParameters=true, module=SetVariables_#11}
   &RegionMask(
    Invert Region=true
    Offset:Vector3f=<0 165 0>
    Offset Space:ENiagaraCoordinateSpace=::Local
    Outer Float=0
    Region Dimensions:Vector3f=<200 850 1500>
    Region Falloff=200
    Region Type:ENiagaraDebugDrawShape=::NewEnumerator1
    Rotation:Vector3f=$Engine.Owner.SystemYAxis
    Rotation Angle Type:ENiagara_AngleInput=::NewEnumerator0
    UseOffset=true
    UseRotation=true
   )
   &Timeline(
    Additional Output:ENiagara_TimelineOutput=::NewEnumerator1
    Additional Output Mode:ENiagara_TimelineOutputMode=::NewEnumerator1
    Float Curve=DataInterface() {propertyValues {Curve {
       defaultValue=0
       keys [
        {interpMode=RCIM_Cubic, tangentMode=RCTM_User, time=0, value=0}
        {arriveTangent=1, interpMode=RCIM_Cubic, leaveTangent=1, time=0.2, value=0}
        {interpMode=RCIM_Cubic, tangentMode=RCTM_User, time=1, value=1}
       ]
      }}}
    Play=&Boolean_NotOperation(Bool=$Particles.RegionMask.IsInsideMask)
    Play Rate=&Lerp_Float(
     A=5
     B=20
     Alpha=&RemapRange(
      Input Value To Remap=&VectorLength(Vector=$User.OwnerLocalVelocity)
      Input Min=0
      Input Max=800
      Output Min=0
      Output Max=1
     )
    )
    Rewind Rate:NiagaraFloat=&Lerp_Float(
     A=5
     B=20
     Alpha=&RemapRange(
      Input Value To Remap=&VectorLength(Vector=$User.OwnerLocalVelocity)
      Input Min=0
      Input Max=800
      Output Min=0
      Output Max=1
     )
    )
   )
   &ScaleMeshSize(Scale Factor=&VectorFromFloat(Value=&Multiply_Float(
      A=$Particles.Timeline.Float
      B=&FloatFromCurve(FloatCurve=DataInterface() {propertyValues {Curve {
          defaultValue=1
          keys [{interpMode=RCIM_Cubic, tangentMode=RCTM_User, time=0, value=1}, {interpMode=RCIM_Cubic, tangentMode=RCTM_User, time=1, value=1}]
         }}})
     ))) {module=ScaleMeshSize001}
   /SetVariables(Particles.ParticleZScaleFactor:NiagaraFloat=&ClampFloat(
     Float=&Multiply_Float(
      A=&self:Lerp_Player_Position_EQ(
       TargetPos:Vector3f=$Particles.TargetPos
       ChangeRate:NiagaraFloat=20
       XY Distance:NiagaraFloat=100
       ParticlePosition:Vector3f=Linked() {linkedVariable {name=Particles.CollisionPos, type=NiagaraPosition}}
       BaseScale:NiagaraFloat=&Lerp_Float(
        A=-200
        B=125
        Alpha=&RemapRange(Input Value To Remap=$Particles.LerpedAudioReactive, Input Min=0.3, Input Max=1, Output Min=0, Output Max=1)
       )
       InputMinTest:NiagaraFloat=1
       TargetZOffset:NiagaraFloat=0
       DeltaTime:NiagaraFloat=&Multiply_Float(A=$Engine.DeltaTime, B=1)
      )
      B=&Lerp_Float(
       A=0.1
       B=0.12
       Alpha=&RemapRange(
        Input Value To Remap=&VectorLength(Vector=$User.OwnerLocalVelocity)
        Input Min=0
        Input Max=800
        Output Min=0
        Output Max=1
       )
      )
     )
     Min=1
     Max=&Lerp_Float(
      A=0
      B=35
      Alpha=&RemapRange(
       Input Value To Remap=&ClampFloat(Float=$Engine.System.Age, Max=0.5)
       Input Min=0
       Input Max=0.5
       Output Min=0
       Output Max=1
      )
     )
    )) {isSetParameters=true, module=SetVariables_#12}
   &ScaleMeshSize(Scale Factor=&MakeVector(X=1, Y=1, Z=$Particles.ParticleZScaleFactor)) {module=ScaleMeshSize002}
   &ScaleMeshSize(Scale Factor=&Lerp_Vector(
     A=<1 1 1>
     B=<1 1 0.5>
     Alpha=&RemapRange(
      Input Value To Remap=&self:XY_Distance_From_Closest_Joint(Target:Vector3f=$Particles.TargetPos, Current:Vector3f=Linked() {linkedVariable {name=Particles.Position, type=NiagaraPosition}})
      Input Min=0
      Input Max=400
      Output Min=0
      Output Max=1
     )
    )) {module=ScaleMeshSize008}
   /SetVariables(Particles.AudioReactiveColorValue:NiagaraFloat=&Multiply_Float(
     A=&Lerp_Float(
      A=1
      B=0.5
      Alpha=&RemapRange(
       Input Value To Remap=&self:XY_Distance_From_Closest_Joint(Target:Vector3f=$Particles.TargetPos, Current:Vector3f=Linked() {linkedVariable {name=Particles.Position, type=NiagaraPosition}})
       Input Min=0
       Input Max=400
       Output Min=0
       Output Max=1
      )
     )
     B=&Lerp_Float(
      A=-0.8
      B=1
      Alpha=&RemapRange(Input Value To Remap=$Particles.LerpedAudioReactive, Input Min=0.3, Input Max=1, Output Min=0, Output Max=1)
     )
    )) {isSetParameters=true, module=SetVariables_#13}
   &ScaleColor(
    Scale Alpha=&Lerp_Float(
     A=1
     B=0
     Alpha=&RemapRange(
      Input Value To Remap=&DistanceBetweenPositions(A=$Particles.Position, B=Linked() {linkedVariable {name=Particles.PelvisPosition, type=Vector3f}})
      Input Min=70
      Input Max=160
      Output Min=0
      Output Max=1
     )
    )
    Scale RGB=&Lerp_Vector(
     A=&Lerp_Vector(
      A=&FN_FrontEndVector3Modifier(
       Gameplay=&MakeVectorFromLinearColorRGB(Color=<rgba 0.1048 5 1.9461 1>)
       Front End=&MakeVectorFromLinearColorRGB(Color=<rgba 0 1.5576 3 1>)
      )
      B=&MakeVectorFromLinearColorRGB(Color=<rgba 0 2.5179 5 1>)
      Alpha=&RemapRange(
       Input Value To Remap=&Sine(
        Normalized Angle=&Add_Float(A=$Emitter.Age, B=&RandomRangeFloat(Minimum=-1, Maximum=1))
        Period=5
        Scale=1
        Bias=0
       )
       Input Min=-1
       Input Max=1
       Output Min=0
       Output Max=0.6
      )
     )
     B=&MakeVectorFromLinearColorRGB(Color=<rgba 0.8642 0.7917 1 1>)
     Alpha=&RemapRange(
      Input Value To Remap=&DistanceBetweenPositions(
       A=&MakeVector(
        X=&MakeFloatFromVector(Vector=Linked() {linkedVariable {name=Particles.Position, type=NiagaraPosition}}, Channel=::NewEnumerator0)
        Y=&MakeFloatFromVector(Vector=Linked() {linkedVariable {name=Particles.Position, type=NiagaraPosition}}, Channel=::NewEnumerator1)
        Z=0
       )
       B=&MakeVector(
        X=&MakeFloatFromVector(Vector=&SimulationPosition(), Channel=::NewEnumerator0)
        Y=&MakeFloatFromVector(Vector=&SimulationPosition(), Channel=::NewEnumerator1)
        Z=0
       )
      )
      Input Min=0
      Input Max=&Lerp_Float(
       A=50
       B=120
       Alpha=&RemapRange(Input Value To Remap=$Particles.LerpedAudioReactive, Input Min=0.6, Input Max=1, Output Min=0, Output Max=1)
      )
      Output Min=0
      Output Max=1
     )
    )
    ScaleA=true
    ScaleRGB=true
   ) {module=ScaleColor002}
   &ScaleColor(
    Scale Alpha=&Lerp_Float(
     A=1
     B=0
     Alpha=&RemapRange(
      Input Value To Remap=&DistanceBetweenPositions(A=$Particles.Position, B=Linked() {linkedVariable {name=Particles.PelvisPosition, type=Vector3f}})
      Input Min=70
      Input Max=160
      Output Min=0
      Output Max=1
     )
    )
    Scale RGB=&Lerp_Vector(
     A=<1 1 1>
     B=&MakeVectorFromLinearColorRGB(Color=<rgba 0.0968 4.6189 1.7978 1>)
     Alpha=&ClampFloat(Float=&Multiply_Float(A=$Particles.AudioReactiveColorValue, B=1))
    )
    ScaleA=true
    ScaleRGB=true
   )
   &ScaleMeshSize(Scale Factor=&Lerp_Vector(
     A=<1 1 1>
     B=&Lerp_Vector(
      A=<1 1 0.8>
      B=<1 1 1.1>
      Alpha=&Lerp_Float(
       A=0
       B=&Add_Float(
        A=&Sine(
         Normalized Angle=&Add_Float(
          A=$Emitter.Age
          B=&Multiply_Float(
           A=&DistanceBetweenPositions(A=$Particles.Position, B=Linked() {linkedVariable {name=Particles.PelvisPosition, type=Vector3f}})
           B=-0.01
          )
         )
         Period=1
         Scale=1
         Bias=0
        )
        B=1
       )
       Alpha=&RemapRange(
        Input Value To Remap=&DistanceBetweenPositions(A=Linked() {linkedVariable {name=Particles.PelvisPosition, type=Vector3f}}, B=$Particles.Position)
        Input Min=0
        Input Max=150
        Output Min=0
        Output Max=1
       )
      )
     )
     Alpha=&self:Height_Gate(
      ChangeRate:NiagaraFloat=12
      Max:NiagaraFloat=0.45
      Particle Z Scale:NiagaraFloat=&MakeFloatFromVector(Vector=$Particles.Scale, Channel=::NewEnumerator2)
     )
    ))
  ]
  rendererClasses [&NiagaraMeshRendererProperties]
  renderers [{rendererClass=&NiagaraMeshRendererProperties, rendererIndex=0}]
 }
}
summary {
 systemName=NS_FareSporeCookie_Final
 emitters [
  {bEnabled=false, emitterName=Visualizer, rendererClasses [&NiagaraMeshRendererProperties]}
  {emitterName=Secondary_Cubes, rendererClasses [&NiagaraMeshRendererProperties]}
  {emitterName=Glow, rendererClasses [&NiagaraSpriteRendererProperties]}
  {emitterName=Light_Down, rendererClasses [&NiagaraComponentRendererProperties]}
  {emitterName=Light_Up, rendererClasses [&NiagaraComponentRendererProperties]}
  {bEnabled=false, emitterName=Visualizer_Low_Med, rendererClasses [&NiagaraMeshRendererProperties]}
  {emitterName=New_Visualizer, rendererClasses [&NiagaraMeshRendererProperties, &NiagaraSpriteRendererProperties]}
 ]
 userVariables [
  User.AltTimeline {type=NiagaraFloat, defaultValue=0, description=""}
  User.bIsFrontEnd {type=NiagaraBool}
  User.bIsFrontEndPreview {type=NiagaraBool}
  User.FrontEnd_color {type=LinearColor, description="", defaultValue {struct=LinearColor, value=<rgba 1 0.6743 0.6445 1>}}
  User.HandGlow {type=NiagaraFloat, defaultValue=0, description=""}
  User.OwnerLocalVelocity {type=Vector3f, description="", defaultValue {struct=Vector3f, value=<1 1 1>}}
  User.SkeletalMesh {
   type=NiagaraDataInterfaceSkeletalMesh
   defaultValue {
    struct=NiagaraExt_VariableValue_DataInterface
    value {dataInterface=&self:NiagaraDataInterfaceSkeletalMesh_0}
   }
  }
 ]
}
